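# Editor utility widget for the World Partition Actor Map tool that shows a searchable list of actor instances from one Actor Map item.
# SetListData stores the item and builds list data from each actor descriptor: label, package, path, and bounds center and extent.
# FilterByString collapses non-matching rows and updates the shown/total count.
# Shift-click loads the actor by GUID via its list index, while plain click focuses the viewport on its center using fixed-size bounds.
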EditorUtilityWidgetBlueprint /Game/Blueprints/Blutilities/ForLDs/WorldPartitionTools/WP_ActorMap/Widgets/WP_ActorMap_List_EUW
graphs [
 EventGraph {nodes [
   n1 = "On Item Left Clicked (TA_Generic01_List)"!() -> (OutputDelegate, then, isShiftDown:bool, isCtrlDown:bool, data:Object Reference, item:User Widget Object Reference) {is=title K2Node_ComponentBoundEvent}
   > n2 = Branch(Condition=@n1.isShiftDown)
   > n3 = getIndexOfItem!(
    execute
    self:TA Generic 01 List Object Reference=Get TA_Generic01_List() -> (TA_Generic01_List:TA Generic 01 List Object Reference)
    Content:Widget Object Reference=@n1.item
   ) -> (then, index:int=0)
   n4 = Break Actor Map Item Struct!(ActorMap_ItemStruct:"Actor Map Item Struct Structure (by ref)"=Get currentData() -> (currentData:Actor Map Item Struct Structure)) -> (
    perInstanceData_12_#1:Array of Actor Map Per Instance Struct Structures
    ItemLabel_10_#2:String
    ItemIcon_14_#3:Texture 2D Soft Object Reference
   ) {is=title K2Node_BreakStruct}
   > LoadActors!(
    execute
    InActorsToLoad:Array of Guid Structures=Make Array("[0]":Guid Structure="Get (a copy)"!(Array:Array of Actor Map Per Instance Struct Structures=@n4, Dimension 1=@n3) -> (
      Output_ActorDesc_4_#1_Guid:Guid Structure
      Output_ActorDesc_4_#1_NativeClass:Object Class Reference
      Output_ActorDesc_4_#1_Class:Soft Object Path Structure
      Output_ActorDesc_4_#1_Name:Name
      Output_ActorDesc_4_#1_Label:Name
      Output_ActorDesc_4_#1_Bounds:Box Structure
      Output_ActorDesc_4_#1_RuntimeGrid:Name
      Output_ActorDesc_4_#1_bIsSpatiallyLoaded:bool=~
      Output_ActorDesc_4_#1_bActorIsEditorOnly:bool=~
      Output_ActorDesc_4_#1_ActorPackage:Name
      Output_ActorDesc_4_#1_ActorPath:Name
      Output_ActorDesc_4_#1_FolderPath:Name
      Output_ActorDesc_4_#1_DataLayerAssets:Array of Soft Object Path Structures
     )) -> (Array:Array of Guid Structures)
   ) -> (then)
   n5 = Cast To TA_GenericListItemData!(execute=@n2.else, Object=@n1.data) -> (then, CastFailed, AsTA Generic List Item Data:TA Generic List Item Data Object Reference) {is=title K2Node_DynamicCast}
   > Focus Bounds in Viewport!(
    " ":Exec
    center:Vector="Get (a copy)"!(Array:Array of Vectors=Get vectors!(self:TA Generic List Item Data Object Reference=@n5) -> (vectors:Array of Vectors), Dimension 1) -> (Output:Vector)
    bounds:Vector="200,200,200"
   ) -> ("  ":Exec) {is=title K2Node_MacroInstance}
   n6 = "On Text Committed (EditableTextBox_44)"!() -> (OutputDelegate, then, Text:"Text (by ref)", CommitMethod:ETextCommit Enum=Default) {is=title K2Node_ComponentBoundEvent}
   > FilterByString!(execute, self, string:String="To String (Text)"(InText=@n6.Text)) -> (then)
   "On Data Changed (TA_Generic01_List)"!() -> (OutputDelegate, then, list:TA Generic 01 List Object Reference, data:Array of Object References) {is=title K2Node_ComponentBoundEvent}
   > FilterByString()
  ]}
 FilterByString {nodes [
   n1 = FilterByString!() -> (then, string:String) {is=title K2Node_FunctionEntry}
   > Set l_filterString(l_filterString:String=@n1) -> (Output_Get:String)
   > n2 = Sequence!(execute) -> (then_0, then_1, then_2:Exec)
   > n3 = getListItems!(execute, self:TA Generic 01 List Object Reference=Get TA_Generic01_List()) -> (then, items:Array of User Widget Object References)
   > Set l_total(l_total:int=Length(TargetArray:Array of User Widget Object References=@n3)) -> (Output_Get:int=0)
   n4 = Branch(execute=@n2.then_1, Condition="Not Equal Exactly (String)"(A=Get l_filterString() -> (l_filterString:String)))
   > n5 = getListItems(self=Get TA_Generic01_List())
   > n6 = For Each Loop(Array:Array of User Widget Object References=@n5) -> (Array Element:User Widget Object Reference)
   > n7 = Cast To TA_Generic01_ListItem!(execute, Object=@n6) -> (then, CastFailed, AsTA Generic 01 List Item:TA Generic 01 List Item Object Reference) {is=title K2Node_DynamicCast}
   > n8 = Cast To TA_GenericListItemData(Object=Get currentListItemData!(self:TA Base Editor Button Widget Object Reference=@n7) -> (currentListItemData:Object Reference))
   > n9 = Branch(Condition=Contains(SearchIn=Get displayString!(self:TA Generic List Item Data Object Reference=@n8) -> (displayString:String), Substring=Get l_filterString()))
   > SetVisibility(self=@n6, InVisibility=Visible)
   > Increment Int(Value=Get l_count() -> (l_count:int=0))
   SetVisibility(execute=@n9.else, self=@n6, InVisibility=Collapsed)
   n10 = getListItems(execute=@n4.else, self=Get TA_Generic01_List())
   > n11 = For Each Loop(Array=@n10)
   Set l_count(execute=@n11.Completed, l_count:int=Get l_total() -> (l_total:int=0)) -> (Output_Get:int=0)
   SetVisibility(execute=@n11.LoopBody, self=@n11, InVisibility=Visible)
   n12 = "To String (Integer)"(InInt=Get l_count() -> (l_count=0))
   n13 = "To String (Integer)"(InInt=Get l_total() -> (l_total=0))
   Set Text(
    execute=@n2.then_2
    Text="To Text (String)"(InString=Append!(A="Showing ", B=@n12, C:String=" / ", D:String=@n13, E:String=" items found.") -> (ReturnValue))
    self=Get txtCount() -> (txtCount:Common Text Object Reference)
   )
  ]}
 SetListData {nodes [
   n1 = SetListData!() -> (then, data:Actor Map Item Struct Structure) {is=title K2Node_FunctionEntry}
   > clearItems(self:TA Generic 01 List Object Reference=Get TA_Generic01_List())
   > Set currentData(currentData:Actor Map Item Struct Structure=@n1) -> (Output_Get:Actor Map Item Struct Structure)
   > n2 = For Each Loop(Array:Array of Actor Map Per Instance Struct Structures=Break Actor Map Item Struct(ActorMap_ItemStruct=@n1)) -> (Array Element:Actor Map Per Instance Struct Structure)
   > n3 = Construct Object from Class(Class=/Game/Blueprints/Blutilities/ForTechArt/Data/TA_GenericListItemData.TA_GenericListItemData_C) -> (ReturnValue:TA Generic List Item Data Object Reference)
   n4 = Break Actor Map Per Instance Struct!(ActorMap_PerInstanceStruct:"Actor Map Per Instance Struct Structure (by ref)"=@n2) -> (
    ActorDesc_4_#1_Guid:Guid Structure
    ActorDesc_4_#1_NativeClass:Object Class Reference
    ActorDesc_4_#1_Class:Soft Object Path Structure
    ActorDesc_4_#1_Name:Name
    ActorDesc_4_#1_Label:Name
    ActorDesc_4_#1_Bounds:Box Structure
    ActorDesc_4_#1_RuntimeGrid:Name
    ActorDesc_4_#1_bIsSpatiallyLoaded:bool=~
    ActorDesc_4_#1_bActorIsEditorOnly:bool=~
    ActorDesc_4_#1_ActorPackage:Name
    ActorDesc_4_#1_ActorPath:Name
    ActorDesc_4_#1_FolderPath:Name
    ActorDesc_4_#1_DataLayerAssets:Array of Soft Object Path Structures
   ) {is=title K2Node_BreakStruct}
   > Set strings!(
    execute
    strings:Array of Strings=Make Array!(
     "[0]":String="To String (Name)"(InName=@n4.ActorDesc_4_#1_ActorPackage)
     "[1]":String="To String (Name)"(InName=@n4.ActorDesc_4_#1_ActorPath)
    ) -> (Array:Array of Strings)
    self:TA Generic List Item Data Object Reference=@n3
   ) -> (then, Output_Get:Array of Strings)
   > Set displayString!(
    execute
    displayString:String="To String (Name)"(InName=@n4.ActorDesc_4_#1_Label)
    self:TA Generic List Item Data Object Reference=@n3
   ) -> (then, Output_Get:String)
   n5 = Break Box!(Box:"Box Structure (by ref)"=@n4.ActorDesc_4_#1_Bounds) -> (Min:Vector=~, Max:Vector=~, IsValid:bool=~) {is=title K2Node_BreakStruct}
   n6 = vector + vector(A=@n5, B=@n5.Max)
   n7 = vector - vector(A=@n5.Max, B=@n5)
   > Set vectors!(
    execute
    vectors:Array of Vectors=Make Array("[0]":Vector=vector / vector(A:Vector=@n6, B=2) -> (ReturnValue:Vector), "[1]":Vector=vector / vector(A=@n7, B=2)) -> (Array:Array of Vectors)
    self:TA Generic List Item Data Object Reference=@n3
   ) -> (then, Output_Get:Array of Vectors)
   > addItem!(execute, self:TA Generic 01 List Object Reference=Get TA_Generic01_List(), dataObject:Object Reference=@n3) -> (then, newItem:User Widget Object Reference, newData:Object Reference)
   Break Actor Map Per Instance Struct!(ActorMap_PerInstanceStruct=@n2) -> (ActorDesc_4_#1:Actor Desc Structure)
  ]}
]
variables [currentData {type=ActorMap_ItemStruct, default {itemIcon=null, itemLabel="", perInstanceData []}}]
widgets {
 info {rootWidgetClass=&VerticalBox, widgetCount=5}
 widgets [
  VerticalBox_0 {widgetClassPath=&VerticalBox, widget=VerticalBox_0}
  EditableTextBox_44 {
   widgetClassPath=&EditableTextBox
   bIsVariable=true
   parent=VerticalBox_0
   widget=EditableTextBox_44
   properties {
    hintText=Search Here...
    widgetStyle {
     backgroundColor {colorUseRule=UseColor_Specified, specifiedColor=<rgba 1 1 1 1>}
     backgroundImageFocused {
      bIsDynamicallyLoaded=false
      drawAs=RoundedBox
      imageType=NoImage
      mirroring=NoMirror
      resourceName=None
      resourceObject=null
      tiling=NoTile
      imageSize=<0 0>
      margin=<ltrb 0 0 0 0>
      outlineSettings {
       bUseBrushTransparency=false
       roundingType=FixedRadius
       width=1
       color {colorUseRule=UseColor_ColorTable, specifiedColor=<rgba 0 0 0 0>}
       cornerRadii=<4 4 4 4>
      }
      tintColor {colorUseRule=UseColor_Specified, specifiedColor=<rgba 0.0048 0.0048 0.0048 1>}
      uVRegion {bIsValid=false, max=<0 0>, min=<0 0>}
     }
     backgroundImageHovered {
      bIsDynamicallyLoaded=false
      drawAs=RoundedBox
      imageType=NoImage
      mirroring=NoMirror
      resourceName=None
      resourceObject=null
      tiling=NoTile
      imageSize=<0 0>
      margin=<ltrb 0 0 0 0>
      outlineSettings {
       bUseBrushTransparency=false
       roundingType=FixedRadius
       width=1
       color {colorUseRule=UseColor_ColorTable, specifiedColor=<rgba 0 0 0 0>}
       cornerRadii=<4 4 4 4>
      }
      tintColor {colorUseRule=UseColor_Specified, specifiedColor=<rgba 0.0048 0.0048 0.0048 1>}
      uVRegion {bIsValid=false, max=<0 0>, min=<0 0>}
     }
     backgroundImageNormal {
      bIsDynamicallyLoaded=false
      drawAs=RoundedBox
      imageType=NoImage
      mirroring=NoMirror
      resourceName=None
      resourceObject=null
      tiling=NoTile
      imageSize=<0 0>
      margin=<ltrb 0 0 0 0>
      outlineSettings {
       bUseBrushTransparency=false
       roundingType=FixedRadius
       width=1
       color {colorUseRule=UseColor_ColorTable, specifiedColor=<rgba 0 0 0 0>}
       cornerRadii=<4 4 4 4>
      }
      tintColor {colorUseRule=UseColor_Specified, specifiedColor=<rgba 0.0048 0.0048 0.0048 1>}
      uVRegion {bIsValid=false, max=<0 0>, min=<0 0>}
     }
     backgroundImageReadOnly {
      bIsDynamicallyLoaded=false
      drawAs=RoundedBox
      imageType=NoImage
      mirroring=NoMirror
      resourceName=None
      resourceObject=null
      tiling=NoTile
      imageSize=<0 0>
      margin=<ltrb 0 0 0 0>
      outlineSettings {
       bUseBrushTransparency=false
       roundingType=FixedRadius
       width=1
       color {colorUseRule=UseColor_ColorTable, specifiedColor=<rgba 0 0 0 0>}
       cornerRadii=<4 4 4 4>
      }
      tintColor {colorUseRule=UseColor_Specified, specifiedColor=<rgba 0.0284 0.0284 0.0284 1>}
      uVRegion {bIsValid=false, max=<0 0>, min=<0 0>}
     }
     focusedForegroundColor {colorUseRule=UseColor_Specified, specifiedColor=<rgba 1 1 1 1>}
     foregroundColor {colorUseRule=UseColor_Specified, specifiedColor=<rgba 0.5271 0.5271 0.5271 1>}
     hScrollBarPadding=<ltrb 0 0 0 0>
     padding=<ltrb 8 3 8 4>
     readOnlyForegroundColor {colorUseRule=UseColor_Foreground, specifiedColor=<rgba 1 0 1 1>}
     scrollBarStyle {
      thickness=8
      draggedThumbImage {
       bIsDynamicallyLoaded=false
       drawAs=RoundedBox
       imageType=NoImage
       mirroring=NoMirror
       resourceName=None
       resourceObject=null
       tiling=NoTile
       imageSize=<0 0>
       margin=<ltrb 0 0 0 0>
       outlineSettings {
        bUseBrushTransparency=false
        roundingType=FixedRadius
        width=0
        color {colorUseRule=UseColor_Specified, specifiedColor=<rgba 0 0 0 0>}
        cornerRadii=<4 4 4 4>
       }
       tintColor {colorUseRule=UseColor_Specified, specifiedColor=<rgba 0.2159 0.2159 0.2159 1>}
       uVRegion {bIsValid=false, max=<0 0>, min=<0 0>}
      }
      horizontalBackgroundImage {
       bIsDynamicallyLoaded=false
       drawAs=NoDrawType
       imageType=NoImage
       mirroring=NoMirror
       resourceName=None
       resourceObject=null
       tiling=NoTile
       imageSize=<0 0>
       margin=<ltrb 0 0 0 0>
       outlineSettings {
        bUseBrushTransparency=false
        roundingType=HalfHeightRadius
        width=0
        color {colorUseRule=UseColor_Specified, specifiedColor=<rgba 0 0 0 0>}
        cornerRadii=<0 0 0 0>
       }
       tintColor {colorUseRule=UseColor_Specified, specifiedColor=<rgba 1 1 1 1>}
       uVRegion {bIsValid=false, max=<0 0>, min=<0 0>}
      }
      horizontalBottomSlotImage {
       bIsDynamicallyLoaded=false
       drawAs=NoDrawType
       imageType=NoImage
       mirroring=NoMirror
       resourceName=None
       resourceObject=null
       tiling=NoTile
       imageSize=<0 0>
       margin=<ltrb 0 0 0 0>
       outlineSettings {
        bUseBrushTransparency=false
        roundingType=HalfHeightRadius
        width=0
        color {colorUseRule=UseColor_Specified, specifiedColor=<rgba 0 0 0 0>}
        cornerRadii=<0 0 0 0>
       }
       tintColor {colorUseRule=UseColor_Specified, specifiedColor=<rgba 1 1 1 1>}
       uVRegion {bIsValid=false, max=<0 0>, min=<0 0>}
      }
      horizontalTopSlotImage {
       bIsDynamicallyLoaded=false
       drawAs=NoDrawType
       imageType=NoImage
       mirroring=NoMirror
       resourceName=None
       resourceObject=null
       tiling=NoTile
       imageSize=<0 0>
       margin=<ltrb 0 0 0 0>
       outlineSettings {
        bUseBrushTransparency=false
        roundingType=HalfHeightRadius
        width=0
        color {colorUseRule=UseColor_Specified, specifiedColor=<rgba 0 0 0 0>}
        cornerRadii=<0 0 0 0>
       }
       tintColor {colorUseRule=UseColor_Specified, specifiedColor=<rgba 1 1 1 1>}
       uVRegion {bIsValid=false, max=<0 0>, min=<0 0>}
      }
      hoveredThumbImage {
       bIsDynamicallyLoaded=false
       drawAs=RoundedBox
       imageType=NoImage
       mirroring=NoMirror
       resourceName=None
       resourceObject=null
       tiling=NoTile
       imageSize=<0 0>
       margin=<ltrb 0 0 0 0>
       outlineSettings {
        bUseBrushTransparency=false
        roundingType=FixedRadius
        width=0
        color {colorUseRule=UseColor_Specified, specifiedColor=<rgba 0 0 0 0>}
        cornerRadii=<4 4 4 4>
       }
       tintColor {colorUseRule=UseColor_Specified, specifiedColor=<rgba 0.2159 0.2159 0.2159 1>}
       uVRegion {bIsValid=false, max=<0 0>, min=<0 0>}
      }
      normalThumbImage {
       bIsDynamicallyLoaded=false
       drawAs=RoundedBox
       imageType=NoImage
       mirroring=NoMirror
       resourceName=None
       resourceObject=null
       tiling=NoTile
       imageSize=<0 0>
       margin=<ltrb 0 0 0 0>
       outlineSettings {
        bUseBrushTransparency=false
        roundingType=FixedRadius
        width=0
        color {colorUseRule=UseColor_Specified, specifiedColor=<rgba 0 0 0 0>}
        cornerRadii=<4 4 4 4>
       }
       tintColor {colorUseRule=UseColor_Specified, specifiedColor=<rgba 0.0953 0.0953 0.0953 1>}
       uVRegion {bIsValid=false, max=<0 0>, min=<0 0>}
      }
      verticalBackgroundImage {
       bIsDynamicallyLoaded=false
       drawAs=NoDrawType
       imageType=NoImage
       mirroring=NoMirror
       resourceName=None
       resourceObject=null
       tiling=NoTile
       imageSize=<0 0>
       margin=<ltrb 0 0 0 0>
       outlineSettings {
        bUseBrushTransparency=false
        roundingType=HalfHeightRadius
        width=0
        color {colorUseRule=UseColor_Specified, specifiedColor=<rgba 0 0 0 0>}
        cornerRadii=<0 0 0 0>
       }
       tintColor {colorUseRule=UseColor_Specified, specifiedColor=<rgba 1 1 1 1>}
       uVRegion {bIsValid=false, max=<0 0>, min=<0 0>}
      }
      verticalBottomSlotImage {
       bIsDynamicallyLoaded=false
       drawAs=NoDrawType
       imageType=NoImage
       mirroring=NoMirror
       resourceName=None
       resourceObject=null
       tiling=NoTile
       imageSize=<0 0>
       margin=<ltrb 0 0 0 0>
       outlineSettings {
        bUseBrushTransparency=false
        roundingType=HalfHeightRadius
        width=0
        color {colorUseRule=UseColor_Specified, specifiedColor=<rgba 0 0 0 0>}
        cornerRadii=<0 0 0 0>
       }
       tintColor {colorUseRule=UseColor_Specified, specifiedColor=<rgba 1 1 1 1>}
       uVRegion {bIsValid=false, max=<0 0>, min=<0 0>}
      }
      verticalTopSlotImage {
       bIsDynamicallyLoaded=false
       drawAs=NoDrawType
       imageType=NoImage
       mirroring=NoMirror
       resourceName=None
       resourceObject=null
       tiling=NoTile
       imageSize=<0 0>
       margin=<ltrb 0 0 0 0>
       outlineSettings {
        bUseBrushTransparency=false
        roundingType=HalfHeightRadius
        width=0
        color {colorUseRule=UseColor_Specified, specifiedColor=<rgba 0 0 0 0>}
        cornerRadii=<0 0 0 0>
       }
       tintColor {colorUseRule=UseColor_Specified, specifiedColor=<rgba 1 1 1 1>}
       uVRegion {bIsValid=false, max=<0 0>, min=<0 0>}
      }
     }
     textStyle {
      overflowPolicy=Clip
      transformPolicy=None
      colorAndOpacity {colorUseRule=UseColor_Foreground, specifiedColor=<rgba 1 0 1 1>}
      font {
       bForceMonospaced=false
       bMaterialIsStencil=false
       fontMaterial=null
       fontObject=&NotoSans
       letterSpacing=0
       monospacedWidth=1
       size=10
       skewAmount=0
       typefaceFontName=NotoSans-Regular
       outlineSettings {
        bApplyOutlineToDropShadows=false
        bMiteredCorners=false
        bSeparateFillAlpha=false
        outlineMaterial=null
        outlineSize=0
        outlineColor=<rgba 0 0 0 1>
       }
      }
      highlightColor {colorUseRule=UseColor_ColorTable, specifiedColor=<rgba 0 0 0 0>}
      highlightShape {
       bIsDynamicallyLoaded=false
       drawAs=Image
       imageType=NoImage
       mirroring=NoMirror
       resourceName=None
       resourceObject=null
       tiling=NoTile
       imageSize=<0 0>
       margin=<ltrb 0 0 0 0>
       outlineSettings {
        bUseBrushTransparency=false
        roundingType=HalfHeightRadius
        width=0
        color {colorUseRule=UseColor_Specified, specifiedColor=<rgba 0 0 0 0>}
        cornerRadii=<0 0 0 0>
       }
       tintColor {colorUseRule=UseColor_ColorTable, specifiedColor=<rgba 0 0 0 0>}
       uVRegion {bIsValid=false, max=<0 0>, min=<0 0>}
      }
      selectedBackgroundColor {colorUseRule=UseColor_ColorTable, specifiedColor=<rgba 0 0 0 0>}
      shadowColorAndOpacity=<rgba 0 0 0 1>
      shadowOffset=<0 0>
      strikeBrush {
       bIsDynamicallyLoaded=false
       drawAs=Image
       imageType=NoImage
       mirroring=NoMirror
       resourceName=None
       resourceObject=null
       tiling=NoTile
       imageSize=<32 32>
       margin=<ltrb 0 0 0 0>
       outlineSettings {
        bUseBrushTransparency=false
        roundingType=HalfHeightRadius
        width=0
        color {colorUseRule=UseColor_Specified, specifiedColor=<rgba 0 0 0 0>}
        cornerRadii=<0 0 0 0>
       }
       tintColor {colorUseRule=UseColor_Specified, specifiedColor=<rgba 1 1 1 1>}
       uVRegion {bIsValid=false, max=<0 0>, min=<0 0>}
      }
      underlineBrush {
       bIsDynamicallyLoaded=false
       drawAs=Image
       imageType=NoImage
       mirroring=NoMirror
       resourceName=None
       resourceObject=null
       tiling=NoTile
       imageSize=<32 32>
       margin=<ltrb 0 0 0 0>
       outlineSettings {
        bUseBrushTransparency=false
        roundingType=HalfHeightRadius
        width=0
        color {colorUseRule=UseColor_Specified, specifiedColor=<rgba 0 0 0 0>}
        cornerRadii=<0 0 0 0>
       }
       tintColor {colorUseRule=UseColor_Specified, specifiedColor=<rgba 1 1 1 1>}
       uVRegion {bIsValid=false, max=<0 0>, min=<0 0>}
      }
     }
     vScrollBarPadding=<ltrb 0 0 0 0>
    }
   }
   slot {class=VerticalBoxSlot, properties {padding=<ltrb 0 0 0 10>}}
  }
  InvalidationBox_0 {
   widgetClassPath=/Script/UMG.InvalidationBox
   bIsVariable=true
   parent=VerticalBox_0
   widget=InvalidationBox_0
   slot {class=VerticalBoxSlot, properties {size {sizeRule=Fill, value=1}}}
  }
  TA_Generic01_List {
   widgetClassPath=/Game/Blueprints/Blutilities/ForTechArt/Widgets/List/TA_Generic01_List
   bIsVariable=true
   parent=InvalidationBox_0
   widget=TA_Generic01_List
   properties {allowMultiSelect=false}
   slot {class=PanelSlot}
  }
  txtCount {
   widgetClassPath=&CommonTextBlock
   bIsVariable=true
   parent=VerticalBox_0
   widget=txtCount
   properties {
    colorAndOpacity {colorUseRule=UseColor_Specified, specifiedColor=<rgba 1 1 1 0.7>}
    font {
     bForceMonospaced=false
     bMaterialIsStencil=false
     fontMaterial=null
     fontObject=&NotoSans
     letterSpacing=0
     monospacedWidth=1
     size=14
     skewAmount=0
     typefaceFontName=NotoSans-Regular
     outlineSettings {
      bApplyOutlineToDropShadows=false
      bMiteredCorners=false
      bSeparateFillAlpha=false
      outlineMaterial=null
      outlineSize=0
      outlineColor=<rgba 0 0 0 1>
     }
    }
    shadowOffset=<0 0>
   }
   slot {class=VerticalBoxSlot}
  }
 ]
}
