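# Editor utility widget for level designers that maps configured actor classes in a World Partition level as world icons and tabbed lists.
# Populate matches actor descriptors to saved MapItems, feeding list tabs and a found-or-spawned transient WP_ActorMap_Actor.
# Container and vehicle-spawner marker icons override configured ones, partly via inline Python CDO reads.
# Selecting the map actor line-traces the cursor to identify the clicked instance; an MPC slider scales icons.
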EditorUtilityWidgetBlueprint /Game/NeverCook/Blutilities/ForLDs/WorldPartitionTools/WP_ActorMap/WP_ActorMap_EUW
graphs [
 AddItemToMap {nodes [
   n1 = AddItemToMap!() -> (
    then
    classForKey:Actor Class Reference
    dataMap:Map of Actor Class References to Actor Map Item Struct Structures
    NewItem:"Actor Map Per Instance Struct Structure (by ref)"
    classForOverrides:Actor Class Reference
    populationData:Actor Map Item Population Struct Structure
   ) {is=title K2Node_FunctionEntry}
   > Set l_classForKey(l_classForKey:Actor Class Reference=@n1) -> (Output_Get:Actor Class Reference)
   > Set l_itemDataMap(l_itemDataMap:Map of Actor Class References to Actor Map Item Struct Structures=@n1.dataMap) -> (Output_Get:Map of Actor Class References to Actor Map Item Struct Structures)
   > Set l_newItem(l_newItem:Actor Map Per Instance Struct Structure=@n1.NewItem) -> (Output_Get:Actor Map Per Instance Struct Structure)
   > Set l_classForOverrides(l_classForOverrides:Actor Class Reference=@n1.classForOverrides) -> (Output_Get:Actor Class Reference)
   > Set l_popData(l_popData:Actor Map Item Population Struct Structure=@n1.populationData) -> (Output_Get:Actor Map Item Population Struct Structure)
   > n2 = Sequence()
   n3 = Find(
    TargetMap:Map of Actor Class References to Actor Map Item Struct Structures=Get l_itemDataMap() -> (l_itemDataMap:Map of Actor Class References to Actor Map Item Struct Structures)
    Key:"Actor Class Reference (by ref)"=Get l_classForKey() -> (l_classForKey:Actor Class Reference)
   ) -> (Value:Actor Map Item Struct Structure)
   > n4 = Branch(Condition=@n3.ReturnValue)
   > n5 = Set l_ItemStruct(l_ItemStruct:Actor Map Item Struct Structure=@n3) -> (Output_Get:Actor Map Item Struct Structure)
   > n6 = Set l_PerInstancedData(l_PerInstancedData:Array of Actor Map Per Instance Struct Structures=Break Actor Map Item Struct!(ActorMap_ItemStruct:"Actor Map Item Struct Structure (by ref)"=@n5) -> (
     perInstanceData_12_#1:Array of Actor Map Per Instance Struct Structures
     ItemLabel_10_#2:String
     ItemIcon_14_#3:Texture 2D Soft Object Reference
    ) {is=title K2Node_BreakStruct}) -> (Output_Get:Array of Actor Map Per Instance Struct Structures)
   > Add(
    TargetArray:Array of Actor Map Per Instance Struct Structures=@n6
    NewItem:"Actor Map Per Instance Struct Structure (by ref)"=Get l_newItem() -> (l_newItem:Actor Map Per Instance Struct Structure)
   ) {is=title K2Node_CallArrayFunction}
   > n7 = Set members in Actor Map Item Struct!(
    execute:Exec
    StructRef:"Actor Map Item Struct Structure (by ref)"=Get l_ItemStruct() -> (l_ItemStruct:Actor Map Item Struct Structure)
    perInstanceData_12_#1:Array of Actor Map Per Instance Struct Structures=Get l_PerInstancedData() -> (l_PerInstancedData:Array of Actor Map Per Instance Struct Structures)
   ) -> (then:Exec, StructOut:"Actor Map Item Struct Structure (by ref)") {is=title K2Node_SetFieldsInStruct}
   > Add!(
    execute
    TargetMap:Map of Actor Class References to Actor Map Item Struct Structures=Get l_itemDataMap()
    Key:"Actor Class Reference (by ref)"=Get l_classForKey()
    Value:"Actor Map Item Struct Structure (by ref)"=@n7
   ) -> (then) {is=title K2Node_CallFunction}
   n8 = Sequence(execute=@n4.else)
   > n9 = GetClassIconOverride!(execute, self, Class:Object Class Reference=Get l_classForOverrides() -> (l_classForOverrides:Actor Class Reference)) -> (then, Icon:Texture 2D Soft Object Reference)
   > n10 = Branch(Condition=IsValidSoftObjectReference(SoftObjectReference=@n9))
   > Set l_icon(l_icon:Texture 2D Soft Object Reference=@n9) -> (Output_Get:Texture 2D Soft Object Reference)
   Set l_icon(
    execute=@n10.else
    l_icon=Break Actor Map Item Population Struct!(ActorMap_ItemPopulationStruct:"Actor Map Item Population Struct Structure (by ref)"=Get l_popData() -> (l_popData:Actor Map Item Population Struct Structure)) -> (
     ItemLabel_10_#2:String
     ItemIcon_15_#3:Texture 2D Soft Object Reference
     ItemClass_17_#4:Actor Class Reference
     Enabled_19_#5:bool=~
    ) {is=title K2Node_BreakStruct}.ItemIcon_15_#3
   )
   n11 = Get l_popData()
   n12 = Get l_newItem()
   Add(
    execute=@n8.then_1
    TargetMap=Get l_itemDataMap()
    Key=Get l_classForKey()
    Value=Make Actor Map Item Struct!(
     perInstanceData_12_#1:Array of Actor Map Per Instance Struct Structures=Make Array("[0]":Actor Map Per Instance Struct Structure=@n12) -> (Array:Array of Actor Map Per Instance Struct Structures)
     ItemLabel_10_#2:String=Break Actor Map Item Population Struct(ActorMap_ItemPopulationStruct=@n11) -> (Enabled_19_#5=~)
     ItemIcon_14_#3:Texture 2D Soft Object Reference=Get l_icon() -> (l_icon:Texture 2D Soft Object Reference)
    ) -> (ActorMap_ItemStruct:Actor Map Item Struct Structure) {is=title K2Node_MakeStruct}
   )
   Set Map of Actor Class References to Actor Map Item Struct Structures!(
    execute:Exec=@n2.then_1
    Target:Map of Actor Class References to Actor Map Item Struct Structures=@n1.dataMap
    Value:Map of Actor Class References to Actor Map Item Struct Structures=Get l_itemDataMap()
   ) -> (then:Exec) {is=title K2Node_VariableSetRef}
  ]}
 EventGraph {nodes [
   n1 = CustomEvent_0!() -> (OutputDelegate, then, SelectionSet:Typed Element Selection Set Object Reference) {is=title K2Node_CustomEvent}
   n2 = HasSelectedObjects!(
    execute
    self:Typed Element Selection Set Object Reference=@n1.SelectionSet
    InRequiredClass:Object Class Reference=/Game/Blueprints/Blutilities/ForLDs/WorldPartitionTools/WP_ActorMap/WP_ActorMap_Actor.WP_ActorMap_Actor_C
   ) -> (then, ReturnValue:bool)
   > n3 = Branch(Condition=@n2)
   HideSelectionDisplay(execute=@n3.else)
   "On Clicked (btnPop)"!() -> (OutputDelegate, then, widget:User Widget Object Reference, event:Pointer Event Structure) {is=title K2Node_ComponentBoundEvent}
   > PopulateMapItems()
   n4 = "On Item Selection Changed (listTabs)"!() -> (
    OutputDelegate
    then
    selectedData:Array of Object References
    selectedWidgets:Array of User Widget Object References
    selectedIndices:Array of Integers
   ) {is=title K2Node_ComponentBoundEvent}
   > Branch(Condition="integer > integer"(A=Length(TargetArray:Array of Integers=@n4.selectedIndices)))
   > SetActiveWidgetIndex(
    self=Get tabContentSwitcher() -> (tabContentSwitcher:Common Animated Switcher Object Reference)
    Index="Get (a copy)"(Array:Array of Integers=@n4.selectedIndices) -> (Output:int)
   )
   n5 = "On Value Changed (sliderScale)"!() -> (OutputDelegate, then, Value:float) {is=title K2Node_ComponentBoundEvent}
   > SetScalarParameterValue!(
    execute
    WorldContextObject
    Collection:Material Parameter Collection Object Reference=/Game/Blueprints/Blutilities/ForLDs/WorldPartitionTools/WP_ActorMap/Materials/DontSaveMe_MPC_ActorMap
    ParameterName:Name=GlobalWorldIconScale
    ParameterValue:float=@n5.Value
   ) -> (then) {is=title K2Node_CallMaterialParameterCollectionFunction}
   "On Clicked (btnResetScale)"!() -> (OutputDelegate, then, widget:User Widget Object Reference, event:Pointer Event Structure) {is=title K2Node_ComponentBoundEvent}
   > SetValue!(execute, self:Slider Object Reference=Get sliderScale() -> (sliderScale:Analog Slider Object Reference), InValue:float=1) -> (then)
   > SetScalarParameterValue(
    Collection=&DontSaveMe_MPC_ActorMap
    ParameterName=GlobalWorldIconScale
    ParameterValue=1
   )
   "On Clicked (btnSettings)"() {is=title K2Node_ComponentBoundEvent}
   n6 = Get panelSettings() -> (panelSettings:Size Box Object Reference)
   > SetVisibility(self=@n6, InVisibility=Select(Option 0=SelfHitTestInvisible, Option 1=Collapsed, Index=IsVisible(self=@n6)))
   n7 = IsVisible(self=Get panelSettings())
   > SetRenderTransformAngle!(
    execute
    self:Widget Object Reference=Get imgSettingsButton() -> (imgSettingsButton:Image Object Reference)
    Angle:float=SelectFloat(A=180, B=0, bPickA=@n7)
   ) -> (then)
   > SetVisibility(
    self=Get txtSettingsBtn() -> (txtSettingsBtn:Common Text Object Reference)
    InVisibility=Select(Option 0=SelfHitTestInvisible, Option 1=Collapsed, Index=@n7)
   )
   "On Property Changed (DetailsView_105)"!() -> (OutputDelegate, then, PropertyName:Name=None) {is=title K2Node_ComponentBoundEvent}
   > SaveGame_Save(self:TA Base Editor Widget Object Reference)
   "On Clicked (btnToggleWorld)"!() -> (OutputDelegate, then, widget:User Widget Object Reference, event:Pointer Event Structure) {is=title K2Node_ComponentBoundEvent}
   > n8 = getMapActor!(execute, self) -> (then, MapActor:WP Actor Map Actor Object Reference)
   n9 = Get DefaultSceneRoot!(self:WP Actor Map Actor Object Reference=@n8) -> (DefaultSceneRoot:Scene Component Object Reference)
   > SetVisibility!(
    execute
    self:Scene Component Object Reference=@n9
    bNewVisibility:bool=NOT Boolean(A=Get bVisible!(self:Scene Component Object Reference=@n9) -> (bVisible:bool))
    bPropagateToChildren:bool=true
   ) -> (then)
   Event Construct()
   > Parent: Construct()
   > n10 = GetSelectionSet(self=Get LevelEditorSubsystem())
   > Bind Event to On Selection Change(self:Typed Element Selection Set Object Reference=@n10, Delegate=@n1) {is=title K2Node_AddDelegate}
   > SetObject(self=Get DetailsView_105() -> (DetailsView_105:Details View Object Reference), NewObject=Self-Reference())
   > n11 = GetScalarParameterValue!(
    execute
    WorldContextObject
    Collection:Material Parameter Collection Object Reference=&DontSaveMe_MPC_ActorMap
    ParameterName:Name=GlobalWorldIconScale
   ) -> (then, ReturnValue:float) {is=title K2Node_CallMaterialParameterCollectionFunction}
   > SetValue(self=Get sliderScale(), InValue=@n11)
   SetVisibility!(
    execute=@n20.then
    self:Widget Object Reference=Get hboxSelection() -> (hboxSelection:Horizontal Box Object Reference)
    InVisibility:ESlateVisibility Enum=SelfHitTestInvisible
   ) -> (then)
   HideSelectionDisplay(execute=@n15.else)
   n12 = Get Cursor World Ray!(execute=@n1.then) -> (then, CameraLocation:Vector, RayOrigin:Vector, RayDirection:Vector, ReturnValue:bool) {is=title K2Node_CallFunction}
   > n13 = GetEditorWorld(self=Get UnrealEditorSubsystem())
   > n14 = Line Trace By Channel!(
    execute
    WorldContextObject:Object Reference=@n13
    Start:Vector=@n12.RayOrigin
    End:Vector=vector + vector(A=@n12.RayOrigin, B=vector * vector(A=@n12.RayDirection, B:float=55555) -> (ReturnValue=~))
    TraceChannel:ETraceTypeQuery Enum=TraceTypeQuery9
    bTraceComplex:bool=false
    ActorsToIgnore:Array of Actor Object References
    DrawDebugType:EDrawDebugTrace Enum
    bIgnoreSelf:bool
    TraceColor:Linear Color Structure
    TraceHitColor:Linear Color Structure
    DrawTime:float
   ) -> (then, OutHit:Hit Result Structure, ReturnValue:bool)
   > n15 = Branch(Condition=@n14.ReturnValue)
   n16 = BreakHitResult!(Hit:"Hit Result Structure (by ref)"=@n14) -> (
    bBlockingHit:bool
    bInitialOverlap:bool
    Time:float
    Distance:float
    Location:Vector
    ImpactPoint:Vector
    Normal:Vector
    ImpactNormal:Vector
    PhysMat:Physical Material Object Reference
    HitActor:Actor Object Reference
    HitComponent:Primitive Component Object Reference
    HitBoneName:Name
    BoneName:Name
    HitItem:int
    ElementIndex:int
    FaceIndex:int
    TraceStart:Vector
    TraceEnd:Vector
   )
   > n17 = Cast To WP_ActorMap_Actor!(execute, Object=@n16.HitActor) -> (then, CastFailed, AsWP Actor Map Actor:WP Actor Map Actor Object Reference) {is=title K2Node_DynamicCast}
   > n18 = Cast To InstancedStaticMeshComponent!(execute, Object=@n16.HitComponent) -> (then, CastFailed, AsInstanced Static Mesh Component:Instanced Static Mesh Component Object Reference) {is=title K2Node_DynamicCast}
   > n19 = GetDataForComponentAndIndex!(
    execute
    self:WP Actor Map Actor Object Reference=@n17
    Component:Instanced Static Mesh Component Object Reference=@n18
    Index:int=@n16.HitItem
   ) -> (then, valid:bool, item:Actor Map Item Struct Structure, instanceData:Actor Map Per Instance Struct Structure)
   > Branch(Condition=@n19)
   > n20 = HandleClickedInstance!(
    execute
    self
    ActorMap_ItemStruct:"Actor Map Item Struct Structure (by ref)"=@n19.item
    ActorMap_PerInstanceStruct:"Actor Map Per Instance Struct Structure (by ref)"=@n19.instanceData
   ) -> (then)
  ]}
 GetClassIconOverride {nodes [
   n1 = GetClassIconOverride!() -> (then, Class:Object Class Reference) {is=title K2Node_FunctionEntry}
   > Set l_class(l_class:Object Class Reference=@n1) -> (Output_Get:Object Class Reference)
   > n2 = Sequence!(execute) -> (then_0, then_1, then_2:Exec)
   > n3 = Cast To BuildingContainer Class!(execute:Exec, Class:Object Class Reference=Get l_class() -> (l_class:Object Class Reference)) -> (then:Exec, CastFailed:Exec, AsBuilding Container:Building Container Class Reference) {is=title K2Node_ClassDynamicCast}
   > Return Node!(
    execute
    Icon:Texture 2D Soft Object Reference=Break Marked Actor Display Info!(MarkedActorDisplayInfo:"Marked Actor Display Info Structure (by ref)"=Get Class Defaults!(Class:Object Class Reference=@n3) -> (
      SearchingSoundCueLoop:Sound Cue Object Reference
      LootRepeatSoundCue:Sound Cue Object Reference
      OnDamageSoundCue:Sound Cue Object Reference
      OnDeathSoundCue:Sound Cue Object Reference
      SearchedMesh:Static Mesh Object Reference
      AdditionalPickupTags:Gameplay Tag Container Structure
      SpawnSourceOverride:EFortPickupSpawnSource Enum
      SearchedMaterialOverrides:Array of Material Interface Object References
      SearchLootTierGroup:Name
      MarkerDisplay:Marked Actor Display Info Structure
      MarkerPositionOffset:Vector 3f
      NumItemsToDropRange:Vector 2f Structure
      SearchBounceRadiusOverride:float=~
      LootNoiseRange:float=~
      LootTestingData:Fort World Item Definition Object Reference
      LootSpawnLocation:Vector 3f
      InstancedLoot_TossDirection:Vector 3f
      InstancedLoot_TossSpeed:float=~
      InstancedLoot_TossConeHalfAngle:float=~
      LootSpawnLocation_Athena:Vector 3f
      LootTossDirection_Athena:Rotator 3f Structure
      LootTossSpeed_Athena:float=~
      LootTossConeHalfAngle_Athena:float=~
      HighestRarity:EFortRarity Enum
      bCanBeMarked:bool=~
      bUseLootProperties_Athena:bool=~
      bAlwaysShowContainer:bool=~
      bDestroyContainerOnSearch:bool=~
      bForceHidePickupMinimapIndicator:bool=~
      bSpawnLootOnInteract:bool=~
      bForceSpawnLootOnDestruction:bool=~
      bForceTossLootOnSpawn:bool=~
      bAlreadySearched:bool=~
      bGivePickupsDirectlyToPlayer:bool=~
      bSkipRollForDestruction:bool=~
      bBuriedTreasure:bool=~
      bHasRaisedTreasure:bool=~
      bRegenerateLoot:bool=~
      bUseLocationForDrop:bool=~
      LootedWeaponsDurabilityModifier:float=~
      SearchSpeed:Curve Table Row Handle Structure
      SearchText:Text
      TimeUntilLootRegenerates:float=~
     ) {is=title K2Node_GetClassDefaults}.MarkerDisplay) -> (
     DisplayName:Text
     Icon:Texture 2D Soft Object Reference
     CustomIndicatorClass:User Widget Soft Class Reference
     CustomDetailsClass:User Widget Soft Class Reference
     PrimaryColor:Linear Color Structure
     Sound:Sound Base Object Reference
     bAttachSound:bool=~
     ScreenClamping:EFortMarkedActorScreenClamping Enum
     SpecialCull:Bitmask Integer
     SpecialDotSizeThreshold:float=~
     SpecialDistance:float=~
     CustomTimeToLive:float=~
     AdditionalData:Marked Actor Display Info Additional Data Object Reference
    ) {is=title K2Node_BreakStruct}.Icon
   )
   n4 = Cast To FortAthenaLivingWorldVehiclePointProvider Class!(execute:Exec=@n2.then_1, Class:Object Class Reference=Get l_class()) -> (
    then:Exec
    CastFailed:Exec
    AsFort Athena Living World Vehicle Point Provider:Fort Athena Living World Vehicle Point Provider Class Reference
   ) {is=title K2Node_ClassDynamicCast}
   > n5 = Execute Python Script!(
    execute
    PythonScript="import unreal\r\nmy_class = unreal.load_class( None, asset_package_name)\r\nout_asset  = unreal.get_default_object(my_class)"
    asset_package_name:String=GetSoftClassPath!(Class:Object Class Reference=@n4) -> (ReturnValue_PathString:String)
   ) -> (then, ReturnValue, out_asset:Object Reference) {inputs [asset_package_name], outputs [out_asset]}
   > n6 = GetEditorProperty!(execute:Exec, Object:Object Reference=@n5.out_asset, PropertyName:Name=PreviewClasses) -> (then:Exec, PropertyValue:Array of Object Class References, ReturnValue:bool)
   > Branch(Condition=Is Valid Index!(TargetArray:Array of Object Class References=@n6, IndexToTest:int) -> (ReturnValue:bool))
   > n7 = Execute Python Script!(
    execute
    PythonScript="import unreal\r\nout_icon = None\r\nmy_class = unreal.load_class( None, asset_class_string)\r\ncdo  = unreal.get_default_object(my_class)"
    asset_class_string:String=GetSoftClassPath(Class="Get (a copy)"(Array:Array of Object Class References=@n6) -> (Output:Object Class Reference))
   ) -> (then, ReturnValue, cdo:Object Reference) {inputs [asset_class_string], outputs [cdo]}
   > Branch(Condition=Does Object Implement Interface!(TestObject:Object Reference=@n7.cdo, Interface:Interface Interface=/Script/FortniteGame.FortMarkableActorInterface) -> (ReturnValue:bool))
   > Set l_assetToFindMarkerDataFor(l_assetToFindMarkerDataFor:Object Reference=@n7.cdo) -> (Output_Get:Object Reference)
   n8 = Get l_assetToFindMarkerDataFor() -> (l_assetToFindMarkerDataFor:Object Reference)
   Is Valid(exec=@n2.then_2, InputObject=@n8)
   > n9 = Execute Python Script!(
    execute
    PythonScript="marker_display = cdo.get_editor_property(\"MarkerDisplay\")\r\nout_icon = marker_display.icon"
    cdo:Object Reference=@n8
   ) -> (then, ReturnValue, out_icon:Texture 2D Soft Object Reference) {inputs [cdo], outputs [out_icon]}
   > Return Node(Icon=@n9.out_icon)
  ]}
 HandleClickedInstance {nodes [
   n1 = HandleClickedInstance!() -> (
    then
    ActorMap_ItemStruct:"Actor Map Item Struct Structure (by ref)"
    ActorMap_PerInstanceStruct:"Actor Map Per Instance Struct Structure (by ref)"
   ) {is=title K2Node_FunctionEntry}
   n2 = Break Actor Map Per Instance Struct!(ActorMap_PerInstanceStruct:"Actor Map Per Instance Struct Structure (by ref)"=@n1.ActorMap_PerInstanceStruct) -> (ActorDesc_4_#1:Actor Desc Structure) {is=title K2Node_BreakStruct}
   > Set currentSelectionDesc(currentSelectionDesc:Actor Desc Structure=@n2) -> (Output_Get:Actor Desc Structure)
   n3 = Break Actor Desc!(ActorDesc:"Actor Desc Structure (by ref)"=@n2) -> (
    Guid:Guid Structure
    NativeClass:Object Class Reference
    Class:Soft Object Path Structure
    Name:Name
    Label:Name
    Bounds:Box Structure
    RuntimeGrid:Name
    bIsSpatiallyLoaded:bool=~
    bActorIsEditorOnly:bool=~
    ActorPackage:Name
    ActorPath:Name
    FolderPath:Name
    DataLayerAssets:Array of Soft Object Path Structures
   ) {is=title K2Node_BreakStruct}
   > "SetText (Text Box)"!(
    execute
    self:Text Box Object Reference=Get txtLabel() -> (txtLabel:Text Box Object Reference)
    InText:Text="To Text (Name)"!(InName:Name=@n3.Label) -> (ReturnValue:Text)
   ) -> (then)
   > "SetText (Text Box)"(
    self=Get txtClass() -> (txtClass:Text Box Object Reference)
    InText="To Text (String)"(InString=BreakSoftObjectPath(InSoftObjectPath=@n3.Class))
   )
   n4 = "To String (Name)"(InName=@n3.ActorPath)
   > "SetText (Text Box)"(
    self=Get txtPath() -> (txtPath:Text Box Object Reference)
    InText="To Text (String)"(InString=Split(SourceString=@n4, InStr=:, SearchDir=FromStart))
   )
   > n5 = LoadAsset_Blocking(Asset=Break Actor Map Item Struct(ActorMap_ItemStruct=@n1).ItemIcon_14_#3)
   > n6 = Cast To Texture2D(Object=@n5)
   > SetBrushFromTexture!(
    execute
    self:Image Object Reference=Get imgIcon() -> (imgIcon:Image Object Reference)
    Texture:Texture 2D Object Reference=@n6
    bMatchSize:bool=false
   ) -> (then)
  ]}
 HideSelectionDisplay {nodes [
   HideSelectionDisplay() {is=title K2Node_FunctionEntry}
   > Set currentSelectionDesc()
   > SetVisibility(self=Get hboxSelection(), InVisibility=Hidden)
  ]}
 IsItemClass {nodes [
   n1 = IsItemClass!() -> (then, TestClass:Actor Class Reference, ItemClasses:Map of Actor Class References to Actor Map Item Population Struct Structures) {is=title K2Node_FunctionEntry}
   > Set l_testClass(l_testClass:Actor Class Reference=@n1) -> (Output_Get:Actor Class Reference)
   > n2 = Keys!(execute, TargetMap:Map of Actor Class References to Actor Map Item Population Struct Structures=@n1.ItemClasses) -> (then, Keys:Array of Actor Class References)
   > Set l_itemClasses(l_itemClasses:Array of Actor Class References=@n2) -> (Output_Get:Array of Actor Class References)
   > n3 = For Each Loop(Array:Array of Actor Class References=Get l_itemClasses() -> (l_itemClasses:Array of Actor Class References)) -> (Array Element:Actor Class Reference)
   Return Node!(execute=@n3.Completed, valid:bool=false, data:Actor Map Item Population Struct Structure, matchedClass:Actor Class Reference)
   > Branch(Condition=ClassIsChildOf(TestClass=Get l_testClass() -> (l_testClass:Actor Class Reference), ParentClass=@n3))
   n4 = Find(
    TargetMap:Map of Actor Class References to Actor Map Item Population Struct Structures=Get ItemClasses() -> (ItemClasses:Map of Actor Class References to Actor Map Item Population Struct Structures)
    Key=@n3
   ) -> (Value:Actor Map Item Population Struct Structure)
   > Branch(Condition=@n4.ReturnValue)
   > Return Node(valid=true, data=@n4, matchedClass=@n3)
  ]}
 PopulateMapItems {nodes [
   PopulateMapItems() {is=title K2Node_FunctionEntry}
   > n1 = Sequence!(execute) -> (then_0, then_1, then_2, then_3:Exec, then_4:Exec)
   > n2 = getMapActor()
   > Set l_mapActor(l_mapActor:WP Actor Map Actor Object Reference=@n2) -> (Output_Get:WP Actor Map Actor Object Reference)
   > ResetMapActor(self:WP Actor Map Actor Object Reference=Get l_mapActor() -> (l_mapActor:WP Actor Map Actor Object Reference))
   > Clear(TargetArray:Array of Actor Map Item Struct Structures=Get CurrentItemData() -> (CurrentItemData:Array of Actor Map Item Struct Structures))
   > ClearChildren(self=Get tabContentSwitcher())
   > clearItems(self:TA Generic 01 List Object Reference=Get listTabs() -> (listTabs:TA Generic 01 List Object Reference))
   n3 = For Each Loop(
    Exec=@n1.then_1
    Array:Array of Actor Map Item Population Struct Structures=Get MapItems() -> (MapItems:Array of Actor Map Item Population Struct Structures)
   ) -> (Array Element:Actor Map Item Population Struct Structure)
   n4 = Break Actor Map Item Population Struct(ActorMap_ItemPopulationStruct=@n3) -> (Enabled_19_#5=~)
   > Branch(Condition=@n4.Enabled_19_#5)
   > Add(
    TargetMap:Map of Actor Class References to Actor Map Item Population Struct Structures=Get l_itemClasses() -> (l_itemClasses:Map of Actor Class References to Actor Map Item Population Struct Structures)
    Key=@n4.ItemClass_17_#4
    Value:"Actor Map Item Population Struct Structure (by ref)"=@n3
   )
   n5 = GetActorDescs!(execute=@n1.then_2) -> (then, OutActorDescs:Array of Actor Desc Structures, ReturnValue:bool)
   > Branch(Condition=@n5.ReturnValue)
   > n6 = For Each Loop(Array:Array of Actor Desc Structures=@n5) -> (Array Element:Actor Desc Structure)
   > Set l_desc(l_desc:Actor Desc Structure=@n6) -> (Output_Get:Actor Desc Structure)
   n7 = BreakSoftObjectPath(InSoftObjectPath=Break Actor Desc(ActorDesc=@n6) -> (bIsSpatiallyLoaded=~, bActorIsEditorOnly=~).Class)
   > n8 = LoadClassAsset_Blocking(AssetClass=To Soft Class Reference(SoftClassPath=MakeSoftClassPath(PathString=@n7)))
   > n9 = Cast To Actor Class!(execute:Exec, Class:Object Class Reference=@n8) -> (then:Exec, CastFailed:Exec, AsActor:Actor Class Reference) {is=title K2Node_ClassDynamicCast}
   > n10 = Set l_class(l_class:Actor Class Reference=@n9) -> (Output_Get:Actor Class Reference)
   > n11 = IsItemClass!(
    execute
    self
    TestClass:Actor Class Reference=@n10
    ItemClasses:Map of Actor Class References to Actor Map Item Population Struct Structures=Get l_itemClasses()
   ) -> (then, valid:bool, data:Actor Map Item Population Struct Structure, matchedClass:Actor Class Reference) {is=title K2Node_CallFunction}
   > Branch(Condition=@n11)
   > Set l_popData(l_popData=@n11.data)
   > Set l_matchedClass(l_matchedClass:Actor Class Reference=@n11.matchedClass) -> (Output_Get:Actor Class Reference)
   > AddItemToMap!(
    execute
    self
    classForKey:Actor Class Reference=Get l_matchedClass() -> (l_matchedClass:Actor Class Reference)
    dataMap:Map of Actor Class References to Actor Map Item Struct Structures=Get l_itemDataMap()
    NewItem:"Actor Map Per Instance Struct Structure (by ref)"=Make Actor Map Per Instance Struct!(ActorDesc_4_#1:Actor Desc Structure=Get l_desc() -> (l_desc:Actor Desc Structure)) -> (ActorMap_PerInstanceStruct:Actor Map Per Instance Struct Structure) {is=title K2Node_MakeStruct}
    classForOverrides:Actor Class Reference=Get l_class() -> (l_class:Actor Class Reference)
    populationData:Actor Map Item Population Struct Structure=Get l_popData()
   ) -> (then) {is=title K2Node_CallFunction}
   n12 = Get l_class()
   > AddItemToMap(
    classForKey=@n12
    dataMap=Get l_actorDataMap() -> (l_actorDataMap:Map of Actor Class References to Actor Map Item Struct Structures)
    NewItem=Make Actor Map Per Instance Struct(ActorDesc_4_#1=Get l_desc())
    classForOverrides=@n12
    populationData=Get l_popData()
   )
   n13 = Values!(execute=@n1.then_3, TargetMap:Map of Actor Class References to Actor Map Item Struct Structures=Get l_actorDataMap()) -> (then, Values:Array of Actor Map Item Struct Structures)
   > n14 = Set CurrentItemData(CurrentItemData:Array of Actor Map Item Struct Structures=@n13) -> (Output_Get:Array of Actor Map Item Struct Structures)
   > SetDataForMapItems!(execute, self:WP Actor Map Actor Object Reference=Get l_mapActor(), ItemData:Array of Actor Map Item Struct Structures=@n14) -> (then)
   n15 = Values(execute=@n1.then_4, TargetMap=Get l_itemDataMap())
   > n16 = For Each Loop(Array:Array of Actor Map Item Struct Structures=@n15) -> (Array Element:Actor Map Item Struct Structure)
   setIndexSelected!(execute=@n16.Completed, self:TA Generic 01 List Object Reference=Get listTabs(), index:int=0, selected:bool=true, throwEvent:bool=true) -> (then)
   n17 = Create Widget!(
    execute:Exec=@n16.LoopBody
    WorldContextObject:Object Reference
    Class:User Widget Class Reference=/Game/Blueprints/Blutilities/ForLDs/WorldPartitionTools/WP_ActorMap/Widgets/WP_ActorMap_List_EUW.WP_ActorMap_List_EUW_C
    OwningPlayer:Player Controller Object Reference
   ) -> (then:Exec, ReturnValue:WP Actor Map List EUW Object Reference)
   > SetListData!(execute, self:WP Actor Map List EUW Object Reference=@n17, data:Actor Map Item Struct Structure=@n16) -> (then)
   > AddChild(self=Get tabContentSwitcher(), Content=@n17)
   > n18 = Construct Object from Class(Class=/Game/Blueprints/Blutilities/ForTechArt/Data/TA_GenericListItemData.TA_GenericListItemData_C) -> (ReturnValue:TA Generic List Item Data Object Reference)
   > Set displayString!(
    execute
    displayString:String=Break Actor Map Item Struct(ActorMap_ItemStruct=@n16).ItemLabel_10_#2
    self:TA Generic List Item Data Object Reference=@n18
   ) -> (then, Output_Get:String)
   > addItem!(execute, self:TA Generic 01 List Object Reference=Get listTabs(), dataObject:Object Reference=@n18) -> (then, newItem:User Widget Object Reference, newData:Object Reference)
   Get l_matchedClass()
  ]}
 SaveGame_PostLoad {nodes [
   n1 = SaveGame_PostLoad!() -> (then, saveGameExisted:bool, saveGame_Object:Save Game Object Reference) {is=title K2Node_FunctionEntry}
   > Branch(Condition=@n1)
   > n2 = Cast To WP_ActorMap_SaveGame!(execute, Object=@n1.saveGame_Object) -> (then, CastFailed, AsWP Actor Map Save Game:WP Actor Map Save Game Object Reference) {is=title K2Node_DynamicCast}
   > Set MapItems(MapItems:Array of Actor Map Item Population Struct Structures=Get PopulationData!(self:WP Actor Map Save Game Object Reference=@n2) -> (PopulationData:Array of Actor Map Item Population Struct Structures)) -> (Output_Get:Array of Actor Map Item Population Struct Structures)
   Return Node!(execute, " ":bool=false)
  ]}
 SaveGame_PrepareToSave {nodes [
   n1 = SaveGame_PrepareToSave!() -> (then, SaveGameObject:Save Game Object Reference) {is=title K2Node_FunctionEntry}
   > n2 = Cast To WP_ActorMap_SaveGame(Object=@n1)
   > Set PopulationData!(
    execute
    PopulationData:Array of Actor Map Item Population Struct Structures=Get MapItems()
    self:WP Actor Map Save Game Object Reference=@n2
   ) -> (then, Output_Get:Array of Actor Map Item Population Struct Structures)
   Return Node(" "=false)
  ]}
 getMapActor {nodes [
   n1 = GetAllLevelActors(self=Get EditorActorSubsystem())
   > n2 = Filter by Class!(
    execute
    TargetArray:Array of Object References=@n1
    ObjectClass:Object Class Reference=&WP_ActorMap_Actor_C
    FilterType:EEditorScriptingFilterType Enum
   ) -> (then, ReturnValue:Array of WP Actor Map Actor Object References)
   n3 = GetEditorWorld(execute=@n7.then, self=Get UnrealEditorSubsystem())
   > n4 = GetAllActorsOfClass!(
    execute
    WorldContextObject:Object Reference=@n3
    ActorClass:Actor Class Reference=&WP_ActorMap_Actor_C
   ) -> (then, OutActors:Array of WP Actor Map Actor Object References)
   n5 = Branch(execute=[@n2.then, @n4.then], Condition="integer > integer"(A=Length(TargetArray:Array of WP Actor Map Actor Object References=@n4)))
   > Return Node!(
    execute
    MapActor:WP Actor Map Actor Object Reference="Get (a copy)"(Array:Array of WP Actor Map Actor Object References=@n4) -> (Output:WP Actor Map Actor Object Reference)
   )
   n6 = SpawnActorFromClass!(
    execute=@n5.else
    ActorClass:Actor Class Reference=&WP_ActorMap_Actor_C
    Location:Vector="0, 0, 0"
    Rotation:Rotator
    bTransient:bool=true
   ) -> (then, ReturnValue:WP Actor Map Actor Object Reference)
   > Return Node(MapActor=@n6)
   n7 = getMapActor() {is=title K2Node_FunctionEntry}
  ]}
]
variables [
 CurrentItemData {type="array<ActorMap_ItemStruct>", default []}
 currentSelectionDesc {
  type=ActorDesc
  default {
   name=None
   class=""
   actorPackage=None
   actorPath=None
   bActorIsEditorOnly=false
   bIsSpatiallyLoaded=false
   folderPath=None
   guid=#6
   label=None
   nativeClass=null
   runtimeGrid=None
   bounds {isValid=false, max=<0 0 0>, min=<0 0 0>}
   dataLayerAssets []
  }
 }
 MapItems {
  type="array<ActorMap_ItemPopulationStruct>"
  default [
   {
    enabled=true
    itemClass=/Game/Athena/Asteria/Environments/BuildingActors/LightCastle/Props/Asteria_LightCastle_Throne_A.Asteria_LightCastle_Throne_A_C
    itemIcon=/BRCosmetics/UI/Foundation/Textures/Icons/Backpacks/T-Icon-Backpacks-572-Chair
    itemLabel=Throne
   }
   {enabled=true, itemClass=/Game/Building/ActorBlueprints/Containers/Tiered_Chest_Athena.Tiered_Chest_Athena_C, itemIcon=null, itemLabel=Chest}
  ]
 }
]
widgets {
 info {
  parentClass=/Game/Blueprints/Blutilities/ForTechArt/Widgets/TA_Base_EditorWidget.TA_Base_EditorWidget_C
  rootWidgetClass=&HorizontalBox
  widgetCount=38
 }
 widgets [
  HorizontalBox_1 {widgetClassPath=&HorizontalBox, widget=HorizontalBox_1}
  VerticalBox_1 {
   widgetClassPath=&VerticalBox
   parent=HorizontalBox_1
   widget=VerticalBox_1
   slot {class=HorizontalBoxSlot, properties {size {sizeRule=Fill, value=1}}}
  }
  HorizontalBox_2 {widgetClassPath=&HorizontalBox, parent=VerticalBox_1, widget=HorizontalBox_2, slot {class=VerticalBoxSlot}}
  btnPop {
   widgetClassPath=&TA_Generic01_Button
   parent=HorizontalBox_2
   widget=btnPop
   properties {label=Populate Data}
   slot {class=HorizontalBoxSlot, properties {size {sizeRule=Fill, value=1}}}
  }
  btnToggleWorld {
   widgetClassPath=&TA_Generic01_Button
   parent=HorizontalBox_2
   widget=btnToggleWorld
   properties {label=Toggle World Icons, buttonColor_Default=<rgba 0.5781 0.0985 0.0177 1>}
   slot {class=HorizontalBoxSlot, properties {padding=<ltrb 10 0 0 0>, size {sizeRule=Fill, value=1}}}
  }
  Border_1 {
   widgetClassPath=&Border
   parent=VerticalBox_1
   widget=Border_1
   properties {brushColor=<rgba 0 0 0 0.5>, padding=<ltrb 0 10 0 0>}
   slot {class=VerticalBoxSlot, properties {size {sizeRule=Fill, value=1}}}
  }
  VerticalBox_3 {
   widgetClassPath=&VerticalBox
   parent=Border_1
   widget=VerticalBox_3
   slot {class=BorderSlot, properties {padding=<ltrb 0 10 0 0>}}
  }
  listTabs {
   widgetClassPath=/Game/Blueprints/Blutilities/ForTechArt/Widgets/List/TA_Generic01_List
   bIsVariable=true
   parent=VerticalBox_3
   widget=listTabs
   properties {
    allowMultiSelect=false
    listItemClass=/Game/Blueprints/Blutilities/ForTechArt/Widgets/Buttons/TA_Generic01_TabItem.TA_Generic01_TabItem_C
    scrollOrientation=Orient_Horizontal
   }
   slot {class=VerticalBoxSlot}
  }
  Image_153 {
   widgetClassPath=&Image
   bIsVariable=true
   parent=VerticalBox_3
   widget=Image_153
   properties {
    brush {
     bIsDynamicallyLoaded=false
     drawAs=Image
     imageType=NoImage
     mirroring=NoMirror
     resourceName=None
     resourceObject=null
     tiling=NoTile
     imageSize=<32 2>
     margin=<ltrb 0 0 0 0>
     outlineSettings {
      bUseBrushTransparency=false
      roundingType=HalfHeightRadius
      width=0
      color {colorUseRule=UseColor_Specified, specifiedColor=<rgba 0 0 0 0>}
      cornerRadii=<0 0 0 0>
     }
     tintColor {colorUseRule=UseColor_Specified, specifiedColor=<rgba 1 1 1 1>}
     uVRegion {bIsValid=false, max=<0 0>, min=<0 0>}
    }
    colorAndOpacity=<rgba 1 1 1 0.5>
   }
   slot {class=VerticalBoxSlot}
  }
  tabContentSwitcher {
   widgetClassPath=/Script/CommonUI.CommonAnimatedSwitcher
   bIsVariable=true
   parent=VerticalBox_3
   widget=tabContentSwitcher
   slot {class=VerticalBoxSlot, properties {size {sizeRule=Fill, value=1}}}
  }
  Border_0 {
   widgetClassPath=&Border
   parent=VerticalBox_1
   widget=Border_0
   properties {brushColor=<rgba 0.026 0.026 0.026 1>, padding=<ltrb 5 5 5 5>}
   slot {class=VerticalBoxSlot, properties {padding=<ltrb 10 10 10 10>}}
  }
  VerticalBox_2 {widgetClassPath=&VerticalBox, parent=Border_0, widget=VerticalBox_2, slot {class=BorderSlot, properties {padding=<ltrb 5 5 5 5>}}}
  txtLabel_1 {
   widgetClassPath=&CommonTextBlock
   bIsVariable=true
   parent=VerticalBox_2
   widget=txtLabel_1
   properties {
    text=Selection Info:
    font {
     bForceMonospaced=false
     bMaterialIsStencil=false
     fontMaterial=null
     fontObject=&NotoSans
     letterSpacing=0
     monospacedWidth=1
     size=16
     skewAmount=0
     typefaceFontName=NotoSans-Regular
     outlineSettings {
      bApplyOutlineToDropShadows=false
      bMiteredCorners=false
      bSeparateFillAlpha=false
      outlineMaterial=null
      outlineSize=0
      outlineColor=<rgba 0 0 0 1>
     }
    }
    shadowOffset=<0 0>
   }
   slot {class=VerticalBoxSlot, properties {verticalAlignment=VAlign_Center, padding=<ltrb 0 0 0 5>, size {sizeRule=Fill, value=1}}}
  }
  hboxSelection {
   widgetClassPath=&HorizontalBox
   bIsVariable=true
   parent=VerticalBox_2
   widget=hboxSelection
   properties {visibility=Hidden}
   slot {class=VerticalBoxSlot, properties {padding=<ltrb 15 0 0 0>}}
  }
  SizeBox_1 {
   widgetClassPath=&SizeBox
   parent=hboxSelection
   widget=SizeBox_1
   properties {bOverride_HeightOverride=true, bOverride_WidthOverride=true, heightOverride=90, widthOverride=90}
   slot {class=HorizontalBoxSlot}
  }
  imgIcon {widgetClassPath=&Image, bIsVariable=true, parent=SizeBox_1, widget=imgIcon, slot {class=SizeBoxSlot}}
  VerticalBox_0 {
   widgetClassPath=&VerticalBox
   parent=hboxSelection
   widget=VerticalBox_0
   slot {class=HorizontalBoxSlot, properties {padding=<ltrb 15 0 0 0>, size {sizeRule=Fill, value=1}}}
  }
  ScaleBox_0 {
   widgetClassPath=&ScaleBox
   parent=VerticalBox_0
   widget=ScaleBox_0
   properties {stretchDirection=DownOnly}
   slot {class=VerticalBoxSlot, properties {verticalAlignment=VAlign_Center, padding=<ltrb 0 0 20 0>, size {sizeRule=Fill, value=1}}}
  }
  txtLabel {
   widgetClassPath=&EditableTextBox
   bIsVariable=true
   parent=ScaleBox_0
   widget=txtLabel
   properties {
    isReadOnly=true
    widgetStyle {
     backgroundColor {colorUseRule=UseColor_Specified, specifiedColor=<rgba 1 1 1 1>}
     backgroundImageFocused {
      bIsDynamicallyLoaded=false
      drawAs=RoundedBox
      imageType=NoImage
      mirroring=NoMirror
      resourceName=None
      resourceObject=null
      tiling=NoTile
      imageSize=<0 0>
      margin=<ltrb 0 0 0 0>
      outlineSettings {
       bUseBrushTransparency=false
       roundingType=FixedRadius
       width=1
       color {colorUseRule=UseColor_ColorTable, specifiedColor=<rgba 0 0 0 0>}
       cornerRadii=<4 4 4 4>
      }
      tintColor {colorUseRule=UseColor_Specified, specifiedColor=<rgba 0.0048 0.0048 0.0048 1>}
      uVRegion {bIsValid=false, max=<0 0>, min=<0 0>}
     }
     backgroundImageHovered {
      bIsDynamicallyLoaded=false
      drawAs=RoundedBox
      imageType=NoImage
      mirroring=NoMirror
      resourceName=None
      resourceObject=null
      tiling=NoTile
      imageSize=<0 0>
      margin=<ltrb 0 0 0 0>
      outlineSettings {
       bUseBrushTransparency=false
       roundingType=FixedRadius
       width=1
       color {colorUseRule=UseColor_ColorTable, specifiedColor=<rgba 0 0 0 0>}
       cornerRadii=<4 4 4 4>
      }
      tintColor {colorUseRule=UseColor_Specified, specifiedColor=<rgba 0.0048 0.0048 0.0048 1>}
      uVRegion {bIsValid=false, max=<0 0>, min=<0 0>}
     }
     backgroundImageNormal {
      bIsDynamicallyLoaded=false
      drawAs=RoundedBox
      imageType=NoImage
      mirroring=NoMirror
      resourceName=None
      resourceObject=null
      tiling=NoTile
      imageSize=<0 0>
      margin=<ltrb 0 0 0 0>
      outlineSettings {
       bUseBrushTransparency=false
       roundingType=FixedRadius
       width=1
       color {colorUseRule=UseColor_ColorTable, specifiedColor=<rgba 0 0 0 0>}
       cornerRadii=<4 4 4 4>
      }
      tintColor {colorUseRule=UseColor_Specified, specifiedColor=<rgba 0.0048 0.0048 0.0048 1>}
      uVRegion {bIsValid=false, max=<0 0>, min=<0 0>}
     }
     backgroundImageReadOnly {
      bIsDynamicallyLoaded=false
      drawAs=RoundedBox
      imageType=NoImage
      mirroring=NoMirror
      resourceName=None
      resourceObject=null
      tiling=NoTile
      imageSize=<0 0>
      margin=<ltrb 0 0 0 0>
      outlineSettings {
       bUseBrushTransparency=false
       roundingType=FixedRadius
       width=1
       color {colorUseRule=UseColor_ColorTable, specifiedColor=<rgba 0 0 0 0>}
       cornerRadii=<4 4 4 4>
      }
      tintColor {colorUseRule=UseColor_Specified, specifiedColor=<rgba 0.0729 0.0729 0.0729 1>}
      uVRegion {bIsValid=false, max=<0 0>, min=<0 0>}
     }
     focusedForegroundColor {colorUseRule=UseColor_Specified, specifiedColor=<rgba 1 1 1 1>}
     foregroundColor {colorUseRule=UseColor_Specified, specifiedColor=<rgba 0.5271 0.5271 0.5271 1>}
     hScrollBarPadding=<ltrb 0 0 0 0>
     padding=<ltrb 8 3 8 4>
     readOnlyForegroundColor {colorUseRule=UseColor_Foreground, specifiedColor=<rgba 1 0 1 1>}
     scrollBarStyle {
      thickness=8
      draggedThumbImage {
       bIsDynamicallyLoaded=false
       drawAs=RoundedBox
       imageType=NoImage
       mirroring=NoMirror
       resourceName=None
       resourceObject=null
       tiling=NoTile
       imageSize=<0 0>
       margin=<ltrb 0 0 0 0>
       outlineSettings {
        bUseBrushTransparency=false
        roundingType=FixedRadius
        width=0
        color {colorUseRule=UseColor_Specified, specifiedColor=<rgba 0 0 0 0>}
        cornerRadii=<4 4 4 4>
       }
       tintColor {colorUseRule=UseColor_Specified, specifiedColor=<rgba 0.2159 0.2159 0.2159 1>}
       uVRegion {bIsValid=false, max=<0 0>, min=<0 0>}
      }
      horizontalBackgroundImage {
       bIsDynamicallyLoaded=false
       drawAs=NoDrawType
       imageType=NoImage
       mirroring=NoMirror
       resourceName=None
       resourceObject=null
       tiling=NoTile
       imageSize=<0 0>
       margin=<ltrb 0 0 0 0>
       outlineSettings {
        bUseBrushTransparency=false
        roundingType=HalfHeightRadius
        width=0
        color {colorUseRule=UseColor_Specified, specifiedColor=<rgba 0 0 0 0>}
        cornerRadii=<0 0 0 0>
       }
       tintColor {colorUseRule=UseColor_Specified, specifiedColor=<rgba 1 1 1 1>}
       uVRegion {bIsValid=false, max=<0 0>, min=<0 0>}
      }
      horizontalBottomSlotImage {
       bIsDynamicallyLoaded=false
       drawAs=NoDrawType
       imageType=NoImage
       mirroring=NoMirror
       resourceName=None
       resourceObject=null
       tiling=NoTile
       imageSize=<0 0>
       margin=<ltrb 0 0 0 0>
       outlineSettings {
        bUseBrushTransparency=false
        roundingType=HalfHeightRadius
        width=0
        color {colorUseRule=UseColor_Specified, specifiedColor=<rgba 0 0 0 0>}
        cornerRadii=<0 0 0 0>
       }
       tintColor {colorUseRule=UseColor_Specified, specifiedColor=<rgba 1 1 1 1>}
       uVRegion {bIsValid=false, max=<0 0>, min=<0 0>}
      }
      horizontalTopSlotImage {
       bIsDynamicallyLoaded=false
       drawAs=NoDrawType
       imageType=NoImage
       mirroring=NoMirror
       resourceName=None
       resourceObject=null
       tiling=NoTile
       imageSize=<0 0>
       margin=<ltrb 0 0 0 0>
       outlineSettings {
        bUseBrushTransparency=false
        roundingType=HalfHeightRadius
        width=0
        color {colorUseRule=UseColor_Specified, specifiedColor=<rgba 0 0 0 0>}
        cornerRadii=<0 0 0 0>
       }
       tintColor {colorUseRule=UseColor_Specified, specifiedColor=<rgba 1 1 1 1>}
       uVRegion {bIsValid=false, max=<0 0>, min=<0 0>}
      }
      hoveredThumbImage {
       bIsDynamicallyLoaded=false
       drawAs=RoundedBox
       imageType=NoImage
       mirroring=NoMirror
       resourceName=None
       resourceObject=null
       tiling=NoTile
       imageSize=<0 0>
       margin=<ltrb 0 0 0 0>
       outlineSettings {
        bUseBrushTransparency=false
        roundingType=FixedRadius
        width=0
        color {colorUseRule=UseColor_Specified, specifiedColor=<rgba 0 0 0 0>}
        cornerRadii=<4 4 4 4>
       }
       tintColor {colorUseRule=UseColor_Specified, specifiedColor=<rgba 0.2159 0.2159 0.2159 1>}
       uVRegion {bIsValid=false, max=<0 0>, min=<0 0>}
      }
      normalThumbImage {
       bIsDynamicallyLoaded=false
       drawAs=RoundedBox
       imageType=NoImage
       mirroring=NoMirror
       resourceName=None
       resourceObject=null
       tiling=NoTile
       imageSize=<0 0>
       margin=<ltrb 0 0 0 0>
       outlineSettings {
        bUseBrushTransparency=false
        roundingType=FixedRadius
        width=0
        color {colorUseRule=UseColor_Specified, specifiedColor=<rgba 0 0 0 0>}
        cornerRadii=<4 4 4 4>
       }
       tintColor {colorUseRule=UseColor_Specified, specifiedColor=<rgba 0.0953 0.0953 0.0953 1>}
       uVRegion {bIsValid=false, max=<0 0>, min=<0 0>}
      }
      verticalBackgroundImage {
       bIsDynamicallyLoaded=false
       drawAs=NoDrawType
       imageType=NoImage
       mirroring=NoMirror
       resourceName=None
       resourceObject=null
       tiling=NoTile
       imageSize=<0 0>
       margin=<ltrb 0 0 0 0>
       outlineSettings {
        bUseBrushTransparency=false
        roundingType=HalfHeightRadius
        width=0
        color {colorUseRule=UseColor_Specified, specifiedColor=<rgba 0 0 0 0>}
        cornerRadii=<0 0 0 0>
       }
       tintColor {colorUseRule=UseColor_Specified, specifiedColor=<rgba 1 1 1 1>}
       uVRegion {bIsValid=false, max=<0 0>, min=<0 0>}
      }
      verticalBottomSlotImage {
       bIsDynamicallyLoaded=false
       drawAs=NoDrawType
       imageType=NoImage
       mirroring=NoMirror
       resourceName=None
       resourceObject=null
       tiling=NoTile
       imageSize=<0 0>
       margin=<ltrb 0 0 0 0>
       outlineSettings {
        bUseBrushTransparency=false
        roundingType=HalfHeightRadius
        width=0
        color {colorUseRule=UseColor_Specified, specifiedColor=<rgba 0 0 0 0>}
        cornerRadii=<0 0 0 0>
       }
       tintColor {colorUseRule=UseColor_Specified, specifiedColor=<rgba 1 1 1 1>}
       uVRegion {bIsValid=false, max=<0 0>, min=<0 0>}
      }
      verticalTopSlotImage {
       bIsDynamicallyLoaded=false
       drawAs=NoDrawType
       imageType=NoImage
       mirroring=NoMirror
       resourceName=None
       resourceObject=null
       tiling=NoTile
       imageSize=<0 0>
       margin=<ltrb 0 0 0 0>
       outlineSettings {
        bUseBrushTransparency=false
        roundingType=HalfHeightRadius
        width=0
        color {colorUseRule=UseColor_Specified, specifiedColor=<rgba 0 0 0 0>}
        cornerRadii=<0 0 0 0>
       }
       tintColor {colorUseRule=UseColor_Specified, specifiedColor=<rgba 1 1 1 1>}
       uVRegion {bIsValid=false, max=<0 0>, min=<0 0>}
      }
     }
     textStyle {
      overflowPolicy=Clip
      transformPolicy=None
      colorAndOpacity {colorUseRule=UseColor_Foreground, specifiedColor=<rgba 1 0 1 1>}
      font {
       bForceMonospaced=false
       bMaterialIsStencil=false
       fontMaterial=null
       fontObject=&NotoSans
       letterSpacing=0
       monospacedWidth=1
       size=10
       skewAmount=0
       typefaceFontName=NotoSans-Regular
       outlineSettings {
        bApplyOutlineToDropShadows=false
        bMiteredCorners=false
        bSeparateFillAlpha=false
        outlineMaterial=null
        outlineSize=0
        outlineColor=<rgba 0 0 0 1>
       }
      }
      highlightColor {colorUseRule=UseColor_ColorTable, specifiedColor=<rgba 0 0 0 0>}
      highlightShape {
       bIsDynamicallyLoaded=false
       drawAs=Image
       imageType=NoImage
       mirroring=NoMirror
       resourceName=None
       resourceObject=null
       tiling=NoTile
       imageSize=<0 0>
       margin=<ltrb 0 0 0 0>
       outlineSettings {
        bUseBrushTransparency=false
        roundingType=HalfHeightRadius
        width=0
        color {colorUseRule=UseColor_Specified, specifiedColor=<rgba 0 0 0 0>}
        cornerRadii=<0 0 0 0>
       }
       tintColor {colorUseRule=UseColor_ColorTable, specifiedColor=<rgba 0 0 0 0>}
       uVRegion {bIsValid=false, max=<0 0>, min=<0 0>}
      }
      selectedBackgroundColor {colorUseRule=UseColor_ColorTable, specifiedColor=<rgba 0 0 0 0>}
      shadowColorAndOpacity=<rgba 0 0 0 1>
      shadowOffset=<0 0>
      strikeBrush {
       bIsDynamicallyLoaded=false
       drawAs=Image
       imageType=NoImage
       mirroring=NoMirror
       resourceName=None
       resourceObject=null
       tiling=NoTile
       imageSize=<32 32>
       margin=<ltrb 0 0 0 0>
       outlineSettings {
        bUseBrushTransparency=false
        roundingType=HalfHeightRadius
        width=0
        color {colorUseRule=UseColor_Specified, specifiedColor=<rgba 0 0 0 0>}
        cornerRadii=<0 0 0 0>
       }
       tintColor {colorUseRule=UseColor_Specified, specifiedColor=<rgba 1 1 1 1>}
       uVRegion {bIsValid=false, max=<0 0>, min=<0 0>}
      }
      underlineBrush {
       bIsDynamicallyLoaded=false
       drawAs=Image
       imageType=NoImage
       mirroring=NoMirror
       resourceName=None
       resourceObject=null
       tiling=NoTile
       imageSize=<32 32>
       margin=<ltrb 0 0 0 0>
       outlineSettings {
        bUseBrushTransparency=false
        roundingType=HalfHeightRadius
        width=0
        color {colorUseRule=UseColor_Specified, specifiedColor=<rgba 0 0 0 0>}
        cornerRadii=<0 0 0 0>
       }
       tintColor {colorUseRule=UseColor_Specified, specifiedColor=<rgba 1 1 1 1>}
       uVRegion {bIsValid=false, max=<0 0>, min=<0 0>}
      }
     }
     vScrollBarPadding=<ltrb 0 0 0 0>
    }
   }
   slot {class=ScaleBoxSlot, properties {horizontalAlignment=HAlign_Fill}}
  }
  ScaleBox_1 {
   widgetClassPath=&ScaleBox
   parent=VerticalBox_0
   widget=ScaleBox_1
   properties {stretchDirection=DownOnly}
   slot {class=VerticalBoxSlot, properties {verticalAlignment=VAlign_Center, padding=<ltrb 0 0 20 0>, size {sizeRule=Fill, value=1}}}
  }
  txtClass {
   widgetClassPath=&EditableTextBox
   bIsVariable=true
   parent=ScaleBox_1
   widget=txtClass
   properties {
    isReadOnly=true
    widgetStyle {
     backgroundColor {colorUseRule=UseColor_Specified, specifiedColor=<rgba 1 1 1 1>}
     backgroundImageFocused {
      bIsDynamicallyLoaded=false
      drawAs=RoundedBox
      imageType=NoImage
      mirroring=NoMirror
      resourceName=None
      resourceObject=null
      tiling=NoTile
      imageSize=<0 0>
      margin=<ltrb 0 0 0 0>
      outlineSettings {
       bUseBrushTransparency=false
       roundingType=FixedRadius
       width=1
       color {colorUseRule=UseColor_ColorTable, specifiedColor=<rgba 0 0 0 0>}
       cornerRadii=<4 4 4 4>
      }
      tintColor {colorUseRule=UseColor_Specified, specifiedColor=<rgba 0.0048 0.0048 0.0048 1>}
      uVRegion {bIsValid=false, max=<0 0>, min=<0 0>}
     }
     backgroundImageHovered {
      bIsDynamicallyLoaded=false
      drawAs=RoundedBox
      imageType=NoImage
      mirroring=NoMirror
      resourceName=None
      resourceObject=null
      tiling=NoTile
      imageSize=<0 0>
      margin=<ltrb 0 0 0 0>
      outlineSettings {
       bUseBrushTransparency=false
       roundingType=FixedRadius
       width=1
       color {colorUseRule=UseColor_ColorTable, specifiedColor=<rgba 0 0 0 0>}
       cornerRadii=<4 4 4 4>
      }
      tintColor {colorUseRule=UseColor_Specified, specifiedColor=<rgba 0.0048 0.0048 0.0048 1>}
      uVRegion {bIsValid=false, max=<0 0>, min=<0 0>}
     }
     backgroundImageNormal {
      bIsDynamicallyLoaded=false
      drawAs=RoundedBox
      imageType=NoImage
      mirroring=NoMirror
      resourceName=None
      resourceObject=null
      tiling=NoTile
      imageSize=<0 0>
      margin=<ltrb 0 0 0 0>
      outlineSettings {
       bUseBrushTransparency=false
       roundingType=FixedRadius
       width=1
       color {colorUseRule=UseColor_ColorTable, specifiedColor=<rgba 0 0 0 0>}
       cornerRadii=<4 4 4 4>
      }
      tintColor {colorUseRule=UseColor_Specified, specifiedColor=<rgba 0.0048 0.0048 0.0048 1>}
      uVRegion {bIsValid=false, max=<0 0>, min=<0 0>}
     }
     backgroundImageReadOnly {
      bIsDynamicallyLoaded=false
      drawAs=RoundedBox
      imageType=NoImage
      mirroring=NoMirror
      resourceName=None
      resourceObject=null
      tiling=NoTile
      imageSize=<0 0>
      margin=<ltrb 0 0 0 0>
      outlineSettings {
       bUseBrushTransparency=false
       roundingType=FixedRadius
       width=1
       color {colorUseRule=UseColor_ColorTable, specifiedColor=<rgba 0 0 0 0>}
       cornerRadii=<4 4 4 4>
      }
      tintColor {colorUseRule=UseColor_Specified, specifiedColor=<rgba 0.0729 0.0729 0.0729 1>}
      uVRegion {bIsValid=false, max=<0 0>, min=<0 0>}
     }
     focusedForegroundColor {colorUseRule=UseColor_Specified, specifiedColor=<rgba 1 1 1 1>}
     foregroundColor {colorUseRule=UseColor_Specified, specifiedColor=<rgba 0.5271 0.5271 0.5271 1>}
     hScrollBarPadding=<ltrb 0 0 0 0>
     padding=<ltrb 8 3 8 4>
     readOnlyForegroundColor {colorUseRule=UseColor_Foreground, specifiedColor=<rgba 1 0 1 1>}
     scrollBarStyle {
      thickness=8
      draggedThumbImage {
       bIsDynamicallyLoaded=false
       drawAs=RoundedBox
       imageType=NoImage
       mirroring=NoMirror
       resourceName=None
       resourceObject=null
       tiling=NoTile
       imageSize=<0 0>
       margin=<ltrb 0 0 0 0>
       outlineSettings {
        bUseBrushTransparency=false
        roundingType=FixedRadius
        width=0
        color {colorUseRule=UseColor_Specified, specifiedColor=<rgba 0 0 0 0>}
        cornerRadii=<4 4 4 4>
       }
       tintColor {colorUseRule=UseColor_Specified, specifiedColor=<rgba 0.2159 0.2159 0.2159 1>}
       uVRegion {bIsValid=false, max=<0 0>, min=<0 0>}
      }
      horizontalBackgroundImage {
       bIsDynamicallyLoaded=false
       drawAs=NoDrawType
       imageType=NoImage
       mirroring=NoMirror
       resourceName=None
       resourceObject=null
       tiling=NoTile
       imageSize=<0 0>
       margin=<ltrb 0 0 0 0>
       outlineSettings {
        bUseBrushTransparency=false
        roundingType=HalfHeightRadius
        width=0
        color {colorUseRule=UseColor_Specified, specifiedColor=<rgba 0 0 0 0>}
        cornerRadii=<0 0 0 0>
       }
       tintColor {colorUseRule=UseColor_Specified, specifiedColor=<rgba 1 1 1 1>}
       uVRegion {bIsValid=false, max=<0 0>, min=<0 0>}
      }
      horizontalBottomSlotImage {
       bIsDynamicallyLoaded=false
       drawAs=NoDrawType
       imageType=NoImage
       mirroring=NoMirror
       resourceName=None
       resourceObject=null
       tiling=NoTile
       imageSize=<0 0>
       margin=<ltrb 0 0 0 0>
       outlineSettings {
        bUseBrushTransparency=false
        roundingType=HalfHeightRadius
        width=0
        color {colorUseRule=UseColor_Specified, specifiedColor=<rgba 0 0 0 0>}
        cornerRadii=<0 0 0 0>
       }
       tintColor {colorUseRule=UseColor_Specified, specifiedColor=<rgba 1 1 1 1>}
       uVRegion {bIsValid=false, max=<0 0>, min=<0 0>}
      }
      horizontalTopSlotImage {
       bIsDynamicallyLoaded=false
       drawAs=NoDrawType
       imageType=NoImage
       mirroring=NoMirror
       resourceName=None
       resourceObject=null
       tiling=NoTile
       imageSize=<0 0>
       margin=<ltrb 0 0 0 0>
       outlineSettings {
        bUseBrushTransparency=false
        roundingType=HalfHeightRadius
        width=0
        color {colorUseRule=UseColor_Specified, specifiedColor=<rgba 0 0 0 0>}
        cornerRadii=<0 0 0 0>
       }
       tintColor {colorUseRule=UseColor_Specified, specifiedColor=<rgba 1 1 1 1>}
       uVRegion {bIsValid=false, max=<0 0>, min=<0 0>}
      }
      hoveredThumbImage {
       bIsDynamicallyLoaded=false
       drawAs=RoundedBox
       imageType=NoImage
       mirroring=NoMirror
       resourceName=None
       resourceObject=null
       tiling=NoTile
       imageSize=<0 0>
       margin=<ltrb 0 0 0 0>
       outlineSettings {
        bUseBrushTransparency=false
        roundingType=FixedRadius
        width=0
        color {colorUseRule=UseColor_Specified, specifiedColor=<rgba 0 0 0 0>}
        cornerRadii=<4 4 4 4>
       }
       tintColor {colorUseRule=UseColor_Specified, specifiedColor=<rgba 0.2159 0.2159 0.2159 1>}
       uVRegion {bIsValid=false, max=<0 0>, min=<0 0>}
      }
      normalThumbImage {
       bIsDynamicallyLoaded=false
       drawAs=RoundedBox
       imageType=NoImage
       mirroring=NoMirror
       resourceName=None
       resourceObject=null
       tiling=NoTile
       imageSize=<0 0>
       margin=<ltrb 0 0 0 0>
       outlineSettings {
        bUseBrushTransparency=false
        roundingType=FixedRadius
        width=0
        color {colorUseRule=UseColor_Specified, specifiedColor=<rgba 0 0 0 0>}
        cornerRadii=<4 4 4 4>
       }
       tintColor {colorUseRule=UseColor_Specified, specifiedColor=<rgba 0.0953 0.0953 0.0953 1>}
       uVRegion {bIsValid=false, max=<0 0>, min=<0 0>}
      }
      verticalBackgroundImage {
       bIsDynamicallyLoaded=false
       drawAs=NoDrawType
       imageType=NoImage
       mirroring=NoMirror
       resourceName=None
       resourceObject=null
       tiling=NoTile
       imageSize=<0 0>
       margin=<ltrb 0 0 0 0>
       outlineSettings {
        bUseBrushTransparency=false
        roundingType=HalfHeightRadius
        width=0
        color {colorUseRule=UseColor_Specified, specifiedColor=<rgba 0 0 0 0>}
        cornerRadii=<0 0 0 0>
       }
       tintColor {colorUseRule=UseColor_Specified, specifiedColor=<rgba 1 1 1 1>}
       uVRegion {bIsValid=false, max=<0 0>, min=<0 0>}
      }
      verticalBottomSlotImage {
       bIsDynamicallyLoaded=false
       drawAs=NoDrawType
       imageType=NoImage
       mirroring=NoMirror
       resourceName=None
       resourceObject=null
       tiling=NoTile
       imageSize=<0 0>
       margin=<ltrb 0 0 0 0>
       outlineSettings {
        bUseBrushTransparency=false
        roundingType=HalfHeightRadius
        width=0
        color {colorUseRule=UseColor_Specified, specifiedColor=<rgba 0 0 0 0>}
        cornerRadii=<0 0 0 0>
       }
       tintColor {colorUseRule=UseColor_Specified, specifiedColor=<rgba 1 1 1 1>}
       uVRegion {bIsValid=false, max=<0 0>, min=<0 0>}
      }
      verticalTopSlotImage {
       bIsDynamicallyLoaded=false
       drawAs=NoDrawType
       imageType=NoImage
       mirroring=NoMirror
       resourceName=None
       resourceObject=null
       tiling=NoTile
       imageSize=<0 0>
       margin=<ltrb 0 0 0 0>
       outlineSettings {
        bUseBrushTransparency=false
        roundingType=HalfHeightRadius
        width=0
        color {colorUseRule=UseColor_Specified, specifiedColor=<rgba 0 0 0 0>}
        cornerRadii=<0 0 0 0>
       }
       tintColor {colorUseRule=UseColor_Specified, specifiedColor=<rgba 1 1 1 1>}
       uVRegion {bIsValid=false, max=<0 0>, min=<0 0>}
      }
     }
     textStyle {
      overflowPolicy=Clip
      transformPolicy=None
      colorAndOpacity {colorUseRule=UseColor_Foreground, specifiedColor=<rgba 1 0 1 1>}
      font {
       bForceMonospaced=false
       bMaterialIsStencil=false
       fontMaterial=null
       fontObject=&NotoSans
       letterSpacing=0
       monospacedWidth=1
       size=10
       skewAmount=0
       typefaceFontName=NotoSans-Regular
       outlineSettings {
        bApplyOutlineToDropShadows=false
        bMiteredCorners=false
        bSeparateFillAlpha=false
        outlineMaterial=null
        outlineSize=0
        outlineColor=<rgba 0 0 0 1>
       }
      }
      highlightColor {colorUseRule=UseColor_ColorTable, specifiedColor=<rgba 0 0 0 0>}
      highlightShape {
       bIsDynamicallyLoaded=false
       drawAs=Image
       imageType=NoImage
       mirroring=NoMirror
       resourceName=None
       resourceObject=null
       tiling=NoTile
       imageSize=<0 0>
       margin=<ltrb 0 0 0 0>
       outlineSettings {
        bUseBrushTransparency=false
        roundingType=HalfHeightRadius
        width=0
        color {colorUseRule=UseColor_Specified, specifiedColor=<rgba 0 0 0 0>}
        cornerRadii=<0 0 0 0>
       }
       tintColor {colorUseRule=UseColor_ColorTable, specifiedColor=<rgba 0 0 0 0>}
       uVRegion {bIsValid=false, max=<0 0>, min=<0 0>}
      }
      selectedBackgroundColor {colorUseRule=UseColor_ColorTable, specifiedColor=<rgba 0 0 0 0>}
      shadowColorAndOpacity=<rgba 0 0 0 1>
      shadowOffset=<0 0>
      strikeBrush {
       bIsDynamicallyLoaded=false
       drawAs=Image
       imageType=NoImage
       mirroring=NoMirror
       resourceName=None
       resourceObject=null
       tiling=NoTile
       imageSize=<32 32>
       margin=<ltrb 0 0 0 0>
       outlineSettings {
        bUseBrushTransparency=false
        roundingType=HalfHeightRadius
        width=0
        color {colorUseRule=UseColor_Specified, specifiedColor=<rgba 0 0 0 0>}
        cornerRadii=<0 0 0 0>
       }
       tintColor {colorUseRule=UseColor_Specified, specifiedColor=<rgba 1 1 1 1>}
       uVRegion {bIsValid=false, max=<0 0>, min=<0 0>}
      }
      underlineBrush {
       bIsDynamicallyLoaded=false
       drawAs=Image
       imageType=NoImage
       mirroring=NoMirror
       resourceName=None
       resourceObject=null
       tiling=NoTile
       imageSize=<32 32>
       margin=<ltrb 0 0 0 0>
       outlineSettings {
        bUseBrushTransparency=false
        roundingType=HalfHeightRadius
        width=0
        color {colorUseRule=UseColor_Specified, specifiedColor=<rgba 0 0 0 0>}
        cornerRadii=<0 0 0 0>
       }
       tintColor {colorUseRule=UseColor_Specified, specifiedColor=<rgba 1 1 1 1>}
       uVRegion {bIsValid=false, max=<0 0>, min=<0 0>}
      }
     }
     vScrollBarPadding=<ltrb 0 0 0 0>
    }
   }
   slot {class=ScaleBoxSlot, properties {horizontalAlignment=HAlign_Fill}}
  }
  ScaleBox_2 {
   widgetClassPath=&ScaleBox
   parent=VerticalBox_0
   widget=ScaleBox_2
   properties {stretchDirection=DownOnly}
   slot {class=VerticalBoxSlot, properties {verticalAlignment=VAlign_Center, padding=<ltrb 0 0 20 0>, size {sizeRule=Fill, value=1}}}
  }
  txtPath {
   widgetClassPath=&EditableTextBox
   bIsVariable=true
   parent=ScaleBox_2
   widget=txtPath
   properties {
    isReadOnly=true
    widgetStyle {
     backgroundColor {colorUseRule=UseColor_Specified, specifiedColor=<rgba 1 1 1 1>}
     backgroundImageFocused {
      bIsDynamicallyLoaded=false
      drawAs=RoundedBox
      imageType=NoImage
      mirroring=NoMirror
      resourceName=None
      resourceObject=null
      tiling=NoTile
      imageSize=<0 0>
      margin=<ltrb 0 0 0 0>
      outlineSettings {
       bUseBrushTransparency=false
       roundingType=FixedRadius
       width=1
       color {colorUseRule=UseColor_ColorTable, specifiedColor=<rgba 0 0 0 0>}
       cornerRadii=<4 4 4 4>
      }
      tintColor {colorUseRule=UseColor_Specified, specifiedColor=<rgba 0.0048 0.0048 0.0048 1>}
      uVRegion {bIsValid=false, max=<0 0>, min=<0 0>}
     }
     backgroundImageHovered {
      bIsDynamicallyLoaded=false
      drawAs=RoundedBox
      imageType=NoImage
      mirroring=NoMirror
      resourceName=None
      resourceObject=null
      tiling=NoTile
      imageSize=<0 0>
      margin=<ltrb 0 0 0 0>
      outlineSettings {
       bUseBrushTransparency=false
       roundingType=FixedRadius
       width=1
       color {colorUseRule=UseColor_ColorTable, specifiedColor=<rgba 0 0 0 0>}
       cornerRadii=<4 4 4 4>
      }
      tintColor {colorUseRule=UseColor_Specified, specifiedColor=<rgba 0.0048 0.0048 0.0048 1>}
      uVRegion {bIsValid=false, max=<0 0>, min=<0 0>}
     }
     backgroundImageNormal {
      bIsDynamicallyLoaded=false
      drawAs=RoundedBox
      imageType=NoImage
      mirroring=NoMirror
      resourceName=None
      resourceObject=null
      tiling=NoTile
      imageSize=<0 0>
      margin=<ltrb 0 0 0 0>
      outlineSettings {
       bUseBrushTransparency=false
       roundingType=FixedRadius
       width=1
       color {colorUseRule=UseColor_ColorTable, specifiedColor=<rgba 0 0 0 0>}
       cornerRadii=<4 4 4 4>
      }
      tintColor {colorUseRule=UseColor_Specified, specifiedColor=<rgba 0.0048 0.0048 0.0048 1>}
      uVRegion {bIsValid=false, max=<0 0>, min=<0 0>}
     }
     backgroundImageReadOnly {
      bIsDynamicallyLoaded=false
      drawAs=RoundedBox
      imageType=NoImage
      mirroring=NoMirror
      resourceName=None
      resourceObject=null
      tiling=NoTile
      imageSize=<0 0>
      margin=<ltrb 0 0 0 0>
      outlineSettings {
       bUseBrushTransparency=false
       roundingType=FixedRadius
       width=1
       color {colorUseRule=UseColor_ColorTable, specifiedColor=<rgba 0 0 0 0>}
       cornerRadii=<4 4 4 4>
      }
      tintColor {colorUseRule=UseColor_Specified, specifiedColor=<rgba 0.0729 0.0729 0.0729 1>}
      uVRegion {bIsValid=false, max=<0 0>, min=<0 0>}
     }
     focusedForegroundColor {colorUseRule=UseColor_Specified, specifiedColor=<rgba 1 1 1 1>}
     foregroundColor {colorUseRule=UseColor_Specified, specifiedColor=<rgba 0.5271 0.5271 0.5271 1>}
     hScrollBarPadding=<ltrb 0 0 0 0>
     padding=<ltrb 8 3 8 4>
     readOnlyForegroundColor {colorUseRule=UseColor_Foreground, specifiedColor=<rgba 1 0 1 1>}
     scrollBarStyle {
      thickness=8
      draggedThumbImage {
       bIsDynamicallyLoaded=false
       drawAs=RoundedBox
       imageType=NoImage
       mirroring=NoMirror
       resourceName=None
       resourceObject=null
       tiling=NoTile
       imageSize=<0 0>
       margin=<ltrb 0 0 0 0>
       outlineSettings {
        bUseBrushTransparency=false
        roundingType=FixedRadius
        width=0
        color {colorUseRule=UseColor_Specified, specifiedColor=<rgba 0 0 0 0>}
        cornerRadii=<4 4 4 4>
       }
       tintColor {colorUseRule=UseColor_Specified, specifiedColor=<rgba 0.2159 0.2159 0.2159 1>}
       uVRegion {bIsValid=false, max=<0 0>, min=<0 0>}
      }
      horizontalBackgroundImage {
       bIsDynamicallyLoaded=false
       drawAs=NoDrawType
       imageType=NoImage
       mirroring=NoMirror
       resourceName=None
       resourceObject=null
       tiling=NoTile
       imageSize=<0 0>
       margin=<ltrb 0 0 0 0>
       outlineSettings {
        bUseBrushTransparency=false
        roundingType=HalfHeightRadius
        width=0
        color {colorUseRule=UseColor_Specified, specifiedColor=<rgba 0 0 0 0>}
        cornerRadii=<0 0 0 0>
       }
       tintColor {colorUseRule=UseColor_Specified, specifiedColor=<rgba 1 1 1 1>}
       uVRegion {bIsValid=false, max=<0 0>, min=<0 0>}
      }
      horizontalBottomSlotImage {
       bIsDynamicallyLoaded=false
       drawAs=NoDrawType
       imageType=NoImage
       mirroring=NoMirror
       resourceName=None
       resourceObject=null
       tiling=NoTile
       imageSize=<0 0>
       margin=<ltrb 0 0 0 0>
       outlineSettings {
        bUseBrushTransparency=false
        roundingType=HalfHeightRadius
        width=0
        color {colorUseRule=UseColor_Specified, specifiedColor=<rgba 0 0 0 0>}
        cornerRadii=<0 0 0 0>
       }
       tintColor {colorUseRule=UseColor_Specified, specifiedColor=<rgba 1 1 1 1>}
       uVRegion {bIsValid=false, max=<0 0>, min=<0 0>}
      }
      horizontalTopSlotImage {
       bIsDynamicallyLoaded=false
       drawAs=NoDrawType
       imageType=NoImage
       mirroring=NoMirror
       resourceName=None
       resourceObject=null
       tiling=NoTile
       imageSize=<0 0>
       margin=<ltrb 0 0 0 0>
       outlineSettings {
        bUseBrushTransparency=false
        roundingType=HalfHeightRadius
        width=0
        color {colorUseRule=UseColor_Specified, specifiedColor=<rgba 0 0 0 0>}
        cornerRadii=<0 0 0 0>
       }
       tintColor {colorUseRule=UseColor_Specified, specifiedColor=<rgba 1 1 1 1>}
       uVRegion {bIsValid=false, max=<0 0>, min=<0 0>}
      }
      hoveredThumbImage {
       bIsDynamicallyLoaded=false
       drawAs=RoundedBox
       imageType=NoImage
       mirroring=NoMirror
       resourceName=None
       resourceObject=null
       tiling=NoTile
       imageSize=<0 0>
       margin=<ltrb 0 0 0 0>
       outlineSettings {
        bUseBrushTransparency=false
        roundingType=FixedRadius
        width=0
        color {colorUseRule=UseColor_Specified, specifiedColor=<rgba 0 0 0 0>}
        cornerRadii=<4 4 4 4>
       }
       tintColor {colorUseRule=UseColor_Specified, specifiedColor=<rgba 0.2159 0.2159 0.2159 1>}
       uVRegion {bIsValid=false, max=<0 0>, min=<0 0>}
      }
      normalThumbImage {
       bIsDynamicallyLoaded=false
       drawAs=RoundedBox
       imageType=NoImage
       mirroring=NoMirror
       resourceName=None
       resourceObject=null
       tiling=NoTile
       imageSize=<0 0>
       margin=<ltrb 0 0 0 0>
       outlineSettings {
        bUseBrushTransparency=false
        roundingType=FixedRadius
        width=0
        color {colorUseRule=UseColor_Specified, specifiedColor=<rgba 0 0 0 0>}
        cornerRadii=<4 4 4 4>
       }
       tintColor {colorUseRule=UseColor_Specified, specifiedColor=<rgba 0.0953 0.0953 0.0953 1>}
       uVRegion {bIsValid=false, max=<0 0>, min=<0 0>}
      }
      verticalBackgroundImage {
       bIsDynamicallyLoaded=false
       drawAs=NoDrawType
       imageType=NoImage
       mirroring=NoMirror
       resourceName=None
       resourceObject=null
       tiling=NoTile
       imageSize=<0 0>
       margin=<ltrb 0 0 0 0>
       outlineSettings {
        bUseBrushTransparency=false
        roundingType=HalfHeightRadius
        width=0
        color {colorUseRule=UseColor_Specified, specifiedColor=<rgba 0 0 0 0>}
        cornerRadii=<0 0 0 0>
       }
       tintColor {colorUseRule=UseColor_Specified, specifiedColor=<rgba 1 1 1 1>}
       uVRegion {bIsValid=false, max=<0 0>, min=<0 0>}
      }
      verticalBottomSlotImage {
       bIsDynamicallyLoaded=false
       drawAs=NoDrawType
       imageType=NoImage
       mirroring=NoMirror
       resourceName=None
       resourceObject=null
       tiling=NoTile
       imageSize=<0 0>
       margin=<ltrb 0 0 0 0>
       outlineSettings {
        bUseBrushTransparency=false
        roundingType=HalfHeightRadius
        width=0
        color {colorUseRule=UseColor_Specified, specifiedColor=<rgba 0 0 0 0>}
        cornerRadii=<0 0 0 0>
       }
       tintColor {colorUseRule=UseColor_Specified, specifiedColor=<rgba 1 1 1 1>}
       uVRegion {bIsValid=false, max=<0 0>, min=<0 0>}
      }
      verticalTopSlotImage {
       bIsDynamicallyLoaded=false
       drawAs=NoDrawType
       imageType=NoImage
       mirroring=NoMirror
       resourceName=None
       resourceObject=null
       tiling=NoTile
       imageSize=<0 0>
       margin=<ltrb 0 0 0 0>
       outlineSettings {
        bUseBrushTransparency=false
        roundingType=HalfHeightRadius
        width=0
        color {colorUseRule=UseColor_Specified, specifiedColor=<rgba 0 0 0 0>}
        cornerRadii=<0 0 0 0>
       }
       tintColor {colorUseRule=UseColor_Specified, specifiedColor=<rgba 1 1 1 1>}
       uVRegion {bIsValid=false, max=<0 0>, min=<0 0>}
      }
     }
     textStyle {
      overflowPolicy=Clip
      transformPolicy=None
      colorAndOpacity {colorUseRule=UseColor_Foreground, specifiedColor=<rgba 1 0 1 1>}
      font {
       bForceMonospaced=false
       bMaterialIsStencil=false
       fontMaterial=null
       fontObject=&NotoSans
       letterSpacing=0
       monospacedWidth=1
       size=10
       skewAmount=0
       typefaceFontName=NotoSans-Regular
       outlineSettings {
        bApplyOutlineToDropShadows=false
        bMiteredCorners=false
        bSeparateFillAlpha=false
        outlineMaterial=null
        outlineSize=0
        outlineColor=<rgba 0 0 0 1>
       }
      }
      highlightColor {colorUseRule=UseColor_ColorTable, specifiedColor=<rgba 0 0 0 0>}
      highlightShape {
       bIsDynamicallyLoaded=false
       drawAs=Image
       imageType=NoImage
       mirroring=NoMirror
       resourceName=None
       resourceObject=null
       tiling=NoTile
       imageSize=<0 0>
       margin=<ltrb 0 0 0 0>
       outlineSettings {
        bUseBrushTransparency=false
        roundingType=HalfHeightRadius
        width=0
        color {colorUseRule=UseColor_Specified, specifiedColor=<rgba 0 0 0 0>}
        cornerRadii=<0 0 0 0>
       }
       tintColor {colorUseRule=UseColor_ColorTable, specifiedColor=<rgba 0 0 0 0>}
       uVRegion {bIsValid=false, max=<0 0>, min=<0 0>}
      }
      selectedBackgroundColor {colorUseRule=UseColor_ColorTable, specifiedColor=<rgba 0 0 0 0>}
      shadowColorAndOpacity=<rgba 0 0 0 1>
      shadowOffset=<0 0>
      strikeBrush {
       bIsDynamicallyLoaded=false
       drawAs=Image
       imageType=NoImage
       mirroring=NoMirror
       resourceName=None
       resourceObject=null
       tiling=NoTile
       imageSize=<32 32>
       margin=<ltrb 0 0 0 0>
       outlineSettings {
        bUseBrushTransparency=false
        roundingType=HalfHeightRadius
        width=0
        color {colorUseRule=UseColor_Specified, specifiedColor=<rgba 0 0 0 0>}
        cornerRadii=<0 0 0 0>
       }
       tintColor {colorUseRule=UseColor_Specified, specifiedColor=<rgba 1 1 1 1>}
       uVRegion {bIsValid=false, max=<0 0>, min=<0 0>}
      }
      underlineBrush {
       bIsDynamicallyLoaded=false
       drawAs=Image
       imageType=NoImage
       mirroring=NoMirror
       resourceName=None
       resourceObject=null
       tiling=NoTile
       imageSize=<32 32>
       margin=<ltrb 0 0 0 0>
       outlineSettings {
        bUseBrushTransparency=false
        roundingType=HalfHeightRadius
        width=0
        color {colorUseRule=UseColor_Specified, specifiedColor=<rgba 0 0 0 0>}
        cornerRadii=<0 0 0 0>
       }
       tintColor {colorUseRule=UseColor_Specified, specifiedColor=<rgba 1 1 1 1>}
       uVRegion {bIsValid=false, max=<0 0>, min=<0 0>}
      }
     }
     vScrollBarPadding=<ltrb 0 0 0 0>
    }
   }
   slot {class=ScaleBoxSlot, properties {horizontalAlignment=HAlign_Fill}}
  }
  btnSettings {
   widgetClassPath=&Button
   parent=HorizontalBox_1
   widget=btnSettings
   properties {widgetStyle {
     clickedSlateSound {resourceObject=null}
     disabled {
      bIsDynamicallyLoaded=false
      drawAs=RoundedBox
      imageType=NoImage
      mirroring=NoMirror
      resourceName=None
      resourceObject=null
      tiling=NoTile
      imageSize=<0 0>
      margin=<ltrb 0 0 0 0>
      outlineSettings {
       bUseBrushTransparency=false
       roundingType=FixedRadius
       width=1
       color {colorUseRule=UseColor_ColorTable, specifiedColor=<rgba 0 0 0 0>}
       cornerRadii=<4 4 4 4>
      }
      tintColor {colorUseRule=UseColor_Specified, specifiedColor=<rgba 0.0395 0.0395 0.0395 1>}
      uVRegion {bIsValid=false, max=<0 0>, min=<0 0>}
     }
     disabledForeground {colorUseRule=UseColor_ColorTable, specifiedColor=<rgba 0 0 0 0>}
     hovered {
      bIsDynamicallyLoaded=false
      drawAs=RoundedBox
      imageType=NoImage
      mirroring=NoMirror
      resourceName=None
      resourceObject=null
      tiling=NoTile
      imageSize=<32 32>
      margin=<ltrb 0 0 0 0>
      outlineSettings {
       bUseBrushTransparency=false
       roundingType=FixedRadius
       width=1
       color {colorUseRule=UseColor_ColorTable, specifiedColor=<rgba 0 0 0 0>}
       cornerRadii=<4 4 4 4>
      }
      tintColor {colorUseRule=UseColor_Specified, specifiedColor=<rgba 0.0953 0.0953 0.0953 1>}
      uVRegion {bIsValid=false, max=<0 0>, min=<0 0>}
     }
     hoveredForeground {colorUseRule=UseColor_ColorTable, specifiedColor=<rgba 0 0 0 0>}
     hoveredSlateSound {resourceObject=null}
     normal {
      bIsDynamicallyLoaded=false
      drawAs=RoundedBox
      imageType=NoImage
      mirroring=NoMirror
      resourceName=None
      resourceObject=null
      tiling=NoTile
      imageSize=<32 32>
      margin=<ltrb 0 0 0 0>
      outlineSettings {
       bUseBrushTransparency=false
       roundingType=FixedRadius
       width=1
       color {colorUseRule=UseColor_ColorTable, specifiedColor=<rgba 0 0 0 0>}
       cornerRadii=<4 4 4 4>
      }
      tintColor {colorUseRule=UseColor_Specified, specifiedColor=<rgba 0.0395 0.0395 0.0395 1>}
      uVRegion {bIsValid=false, max=<0 0>, min=<0 0>}
     }
     normalForeground {colorUseRule=UseColor_ColorTable, specifiedColor=<rgba 0 0 0 0>}
     normalPadding=<ltrb 12 1.5 12 1.5>
     pressed {
      bIsDynamicallyLoaded=false
      drawAs=RoundedBox
      imageType=NoImage
      mirroring=NoMirror
      resourceName=None
      resourceObject=null
      tiling=NoTile
      imageSize=<32 32>
      margin=<ltrb 0 0 0 0>
      outlineSettings {
       bUseBrushTransparency=false
       roundingType=FixedRadius
       width=1
       color {colorUseRule=UseColor_ColorTable, specifiedColor=<rgba 0 0 0 0>}
       cornerRadii=<4 4 4 4>
      }
      tintColor {colorUseRule=UseColor_Specified, specifiedColor=<rgba 0.0284 0.0284 0.0284 1>}
      uVRegion {bIsValid=false, max=<0 0>, min=<0 0>}
     }
     pressedForeground {colorUseRule=UseColor_ColorTable, specifiedColor=<rgba 0 0 0 0>}
     pressedPadding=<ltrb 12 2.5 12 0.5>
     pressedSlateSound {resourceObject=null}
    }}
   slot {class=HorizontalBoxSlot}
  }
  Overlay_0 {
   widgetClassPath=&Overlay
   parent=btnSettings
   widget=Overlay_0
   slot {class=ButtonSlot, properties {verticalAlignment=VAlign_Fill, padding=<ltrb 0 0 0 0>}}
  }
  imgSettingsButton {
   widgetClassPath=&Image
   bIsVariable=true
   parent=Overlay_0
   widget=imgSettingsButton
   properties {
    brush {
     bIsDynamicallyLoaded=false
     drawAs=Image
     imageType=NoImage
     mirroring=NoMirror
     resourceName=None
     resourceObject=/Game/UI/Style/Button-BasicArrowLeft-_normal_
     tiling=NoTile
     imageSize=<4 8>
     margin=<ltrb 0 0 0 0>
     outlineSettings {
      bUseBrushTransparency=false
      roundingType=HalfHeightRadius
      width=0
      color {colorUseRule=UseColor_Specified, specifiedColor=<rgba 0 0 0 0>}
      cornerRadii=<0 0 0 0>
     }
     tintColor {colorUseRule=UseColor_Specified, specifiedColor=<rgba 1 1 1 1>}
     uVRegion {bIsValid=false, max=<0 0>, min=<0 0>}
    }
    renderTransform {angle=0, scale=<5 10>, shear=<0 0>, translation=<0 0>}
   }
   slot {class=OverlaySlot, properties {verticalAlignment=VAlign_Center}}
  }
  SizeBox_3 {
   widgetClassPath=&SizeBox
   parent=Overlay_0
   widget=SizeBox_3
   properties {
    bOverride_MaxDesiredWidth=true
    bOverride_MinDesiredHeight=true
    maxDesiredHeight=1463.6353
    maxDesiredWidth=2.0239
    minDesiredHeight=177.895
   }
   slot {class=OverlaySlot, properties {horizontalAlignment=HAlign_Center}}
  }
  txtSettingsBtn {
   widgetClassPath=&CommonTextBlock
   bIsVariable=true
   parent=SizeBox_3
   widget=txtSettingsBtn
   properties {
    text=SETTINGS
    font {
     bForceMonospaced=false
     bMaterialIsStencil=false
     fontMaterial=null
     fontObject=&NotoSans
     letterSpacing=616
     monospacedWidth=1
     size=12
     skewAmount=0
     typefaceFontName=NotoSans-Bold
     outlineSettings {
      bApplyOutlineToDropShadows=false
      bMiteredCorners=false
      bSeparateFillAlpha=false
      outlineMaterial=null
      outlineSize=0
      outlineColor=<rgba 0 0 0 1>
     }
    }
    renderTransform {angle=90, scale=<1 1>, shear=<0 0>, translation=<0 0>}
    shadowOffset=<0 0>
   }
   slot {class=SizeBoxSlot, properties {horizontalAlignment=HAlign_Center, verticalAlignment=VAlign_Top}}
  }
  panelSettings {
   widgetClassPath=&SizeBox
   bIsVariable=true
   parent=HorizontalBox_1
   widget=panelSettings
   properties {bOverride_WidthOverride=true, visibility=Collapsed, widthOverride=356.0884}
   slot {class=HorizontalBoxSlot}
  }
  VerticalBox_4 {widgetClassPath=&VerticalBox, parent=panelSettings, widget=VerticalBox_4, slot {class=SizeBoxSlot}}
  Border_2 {
   widgetClassPath=&Border
   parent=VerticalBox_4
   widget=Border_2
   properties {brushColor=<rgba 0 0 0 0.5>, padding=<ltrb 3 3 3 3>}
   slot {class=VerticalBoxSlot, properties {padding=<ltrb 0 0 0 10>}}
  }
  txtSettingsHeader {
   widgetClassPath=&CommonTextBlock
   parent=Border_2
   widget=txtSettingsHeader
   properties {
    text=Settings
    font {
     bForceMonospaced=false
     bMaterialIsStencil=false
     fontMaterial=null
     fontObject=&NotoSans
     letterSpacing=0
     monospacedWidth=1
     size=16
     skewAmount=0
     typefaceFontName=NotoSans-Regular
     outlineSettings {
      bApplyOutlineToDropShadows=false
      bMiteredCorners=false
      bSeparateFillAlpha=false
      outlineMaterial=null
      outlineSize=0
      outlineColor=<rgba 0 0 0 1>
     }
    }
    shadowOffset=<0 0>
   }
   slot {class=BorderSlot, properties {padding=<ltrb 3 3 3 3>}}
  }
  HorizontalBox_0 {widgetClassPath=&HorizontalBox, parent=VerticalBox_4, widget=HorizontalBox_0, slot {class=VerticalBoxSlot}}
  txtSettingsHeader_1 {
   widgetClassPath=&CommonTextBlock
   parent=HorizontalBox_0
   widget=txtSettingsHeader_1
   properties {
    text="Scale: "
    font {
     bForceMonospaced=false
     bMaterialIsStencil=false
     fontMaterial=null
     fontObject=&NotoSans
     letterSpacing=0
     monospacedWidth=1
     size=16
     skewAmount=0
     typefaceFontName=NotoSans-Regular
     outlineSettings {
      bApplyOutlineToDropShadows=false
      bMiteredCorners=false
      bSeparateFillAlpha=false
      outlineMaterial=null
      outlineSize=0
      outlineColor=<rgba 0 0 0 1>
     }
    }
    shadowOffset=<0 0>
   }
   slot {class=HorizontalBoxSlot}
  }
  sliderScale {
   widgetClassPath=/Script/CommonUI.AnalogSlider
   bIsVariable=true
   parent=HorizontalBox_0
   widget=sliderScale
   properties {
    maxValue=2
    minValue=0.1
    requiresControllerLock=false
    value=1
    widgetStyle {
     barThickness=9
     disabledBarImage {
      bIsDynamicallyLoaded=false
      drawAs=RoundedBox
      imageType=NoImage
      mirroring=NoMirror
      resourceName=None
      resourceObject=null
      tiling=NoTile
      imageSize=<32 32>
      margin=<ltrb 0 0 0 0>
      outlineSettings {
       bUseBrushTransparency=false
       roundingType=FixedRadius
       width=1
       color {colorUseRule=UseColor_ColorTable, specifiedColor=<rgba 0 0 0 0>}
       cornerRadii=<2 2 2 2>
      }
      tintColor {colorUseRule=UseColor_Specified, specifiedColor=<rgba 1 1 1 1>}
      uVRegion {bIsValid=false, max=<0 0>, min=<0 0>}
     }
     disabledThumbImage {
      bIsDynamicallyLoaded=false
      drawAs=RoundedBox
      imageType=NoImage
      mirroring=NoMirror
      resourceName=None
      resourceObject=null
      tiling=NoTile
      imageSize=<32 32>
      margin=<ltrb 0 0 0 0>
      outlineSettings {
       bUseBrushTransparency=false
       roundingType=HalfHeightRadius
       width=0
       color {colorUseRule=UseColor_Specified, specifiedColor=<rgba 0 0 0 0>}
       cornerRadii=<0 0 0 0>
      }
      tintColor {colorUseRule=UseColor_Specified, specifiedColor=<rgba 0.5271 0.5271 0.5271 1>}
      uVRegion {bIsValid=false, max=<0 0>, min=<0 0>}
     }
     hoveredBarImage {
      bIsDynamicallyLoaded=false
      drawAs=RoundedBox
      imageType=NoImage
      mirroring=NoMirror
      resourceName=None
      resourceObject=null
      tiling=NoTile
      imageSize=<0 0>
      margin=<ltrb 0 0 0 0>
      outlineSettings {
       bUseBrushTransparency=false
       roundingType=FixedRadius
       width=1
       color {colorUseRule=UseColor_ColorTable, specifiedColor=<rgba 0 0 0 0>}
       cornerRadii=<2 2 2 2>
      }
      tintColor {colorUseRule=UseColor_Specified, specifiedColor=<rgba 0.0048 0.0048 0.0048 1>}
      uVRegion {bIsValid=false, max=<0 0>, min=<0 0>}
     }
     hoveredThumbImage {
      bIsDynamicallyLoaded=false
      drawAs=RoundedBox
      imageType=NoImage
      mirroring=NoMirror
      resourceName=None
      resourceObject=null
      tiling=NoTile
      imageSize=<8 8>
      margin=<ltrb 0 0 0 0>
      outlineSettings {
       bUseBrushTransparency=false
       roundingType=HalfHeightRadius
       width=0
       color {colorUseRule=UseColor_Specified, specifiedColor=<rgba 0 0 0 0>}
       cornerRadii=<0 0 0 0>
      }
      tintColor {colorUseRule=UseColor_Specified, specifiedColor=<rgba 1 1 1 1>}
      uVRegion {bIsValid=false, max=<0 0>, min=<0 0>}
     }
     normalBarImage {
      bIsDynamicallyLoaded=false
      drawAs=RoundedBox
      imageType=NoImage
      mirroring=NoMirror
      resourceName=None
      resourceObject=null
      tiling=NoTile
      imageSize=<0 0>
      margin=<ltrb 0 0 0 0>
      outlineSettings {
       bUseBrushTransparency=false
       roundingType=FixedRadius
       width=1
       color {colorUseRule=UseColor_ColorTable, specifiedColor=<rgba 0 0 0 0>}
       cornerRadii=<2 2 2 2>
      }
      tintColor {colorUseRule=UseColor_Specified, specifiedColor=<rgba 0.0048 0.0048 0.0048 1>}
      uVRegion {bIsValid=false, max=<0 0>, min=<0 0>}
     }
     normalThumbImage {
      bIsDynamicallyLoaded=false
      drawAs=RoundedBox
      imageType=NoImage
      mirroring=NoMirror
      resourceName=None
      resourceObject=null
      tiling=NoTile
      imageSize=<8 8>
      margin=<ltrb 0 0 0 0>
      outlineSettings {
       bUseBrushTransparency=false
       roundingType=HalfHeightRadius
       width=0
       color {colorUseRule=UseColor_Specified, specifiedColor=<rgba 0 0 0 0>}
       cornerRadii=<0 0 0 0>
      }
      tintColor {colorUseRule=UseColor_Specified, specifiedColor=<rgba 0.5271 0.5271 0.5271 1>}
      uVRegion {bIsValid=false, max=<0 0>, min=<0 0>}
     }
    }
   }
   slot {class=HorizontalBoxSlot, properties {size {sizeRule=Fill, value=1}}}
  }
  btnResetScale {
   widgetClassPath=&TA_Generic01_Button
   parent=HorizontalBox_0
   widget=btnResetScale
   properties {
    toolTipText=Reset Scale to 1
    buttonColor_Default=<rgba 1 0.0695 0 1>
    iconBrush {
     bIsDynamicallyLoaded=false
     drawAs=Image
     imageType=NoImage
     mirroring=NoMirror
     resourceName=None
     resourceObject=/Game/Athena/HUD/Art/ClearPersistenceIslandDataIcon
     tiling=NoTile
     imageSize=<16 16>
     margin=<ltrb 0 0 0 0>
     outlineSettings {
      bUseBrushTransparency=false
      roundingType=HalfHeightRadius
      width=0
      color {colorUseRule=UseColor_Specified, specifiedColor=<rgba 0 0 0 0>}
      cornerRadii=<0 0 0 1>
     }
     tintColor {colorUseRule=UseColor_Specified, specifiedColor=<rgba 1 1 1 1>}
     uVRegion {bIsValid=false, max=<0 0>, min=<0 0>}
    }
   }
   slot {class=HorizontalBoxSlot, properties {horizontalAlignment=HAlign_Left, verticalAlignment=VAlign_Center}}
  }
  ScrollBox_0 {
   widgetClassPath=&ScrollBox
   parent=VerticalBox_4
   widget=ScrollBox_0
   properties {
    widgetBarStyle {
     thickness=8
     draggedThumbImage {
      bIsDynamicallyLoaded=false
      drawAs=RoundedBox
      imageType=NoImage
      mirroring=NoMirror
      resourceName=None
      resourceObject=null
      tiling=NoTile
      imageSize=<0 0>
      margin=<ltrb 0 0 0 0>
      outlineSettings {
       bUseBrushTransparency=false
       roundingType=FixedRadius
       width=0
       color {colorUseRule=UseColor_Specified, specifiedColor=<rgba 0 0 0 0>}
       cornerRadii=<4 4 4 4>
      }
      tintColor {colorUseRule=UseColor_Specified, specifiedColor=<rgba 0.2159 0.2159 0.2159 1>}
      uVRegion {bIsValid=false, max=<0 0>, min=<0 0>}
     }
     horizontalBackgroundImage {
      bIsDynamicallyLoaded=false
      drawAs=NoDrawType
      imageType=NoImage
      mirroring=NoMirror
      resourceName=None
      resourceObject=null
      tiling=NoTile
      imageSize=<0 0>
      margin=<ltrb 0 0 0 0>
      outlineSettings {
       bUseBrushTransparency=false
       roundingType=HalfHeightRadius
       width=0
       color {colorUseRule=UseColor_Specified, specifiedColor=<rgba 0 0 0 0>}
       cornerRadii=<0 0 0 0>
      }
      tintColor {colorUseRule=UseColor_Specified, specifiedColor=<rgba 1 1 1 1>}
      uVRegion {bIsValid=false, max=<0 0>, min=<0 0>}
     }
     horizontalBottomSlotImage {
      bIsDynamicallyLoaded=false
      drawAs=NoDrawType
      imageType=NoImage
      mirroring=NoMirror
      resourceName=None
      resourceObject=null
      tiling=NoTile
      imageSize=<0 0>
      margin=<ltrb 0 0 0 0>
      outlineSettings {
       bUseBrushTransparency=false
       roundingType=HalfHeightRadius
       width=0
       color {colorUseRule=UseColor_Specified, specifiedColor=<rgba 0 0 0 0>}
       cornerRadii=<0 0 0 0>
      }
      tintColor {colorUseRule=UseColor_Specified, specifiedColor=<rgba 1 1 1 1>}
      uVRegion {bIsValid=false, max=<0 0>, min=<0 0>}
     }
     horizontalTopSlotImage {
      bIsDynamicallyLoaded=false
      drawAs=NoDrawType
      imageType=NoImage
      mirroring=NoMirror
      resourceName=None
      resourceObject=null
      tiling=NoTile
      imageSize=<0 0>
      margin=<ltrb 0 0 0 0>
      outlineSettings {
       bUseBrushTransparency=false
       roundingType=HalfHeightRadius
       width=0
       color {colorUseRule=UseColor_Specified, specifiedColor=<rgba 0 0 0 0>}
       cornerRadii=<0 0 0 0>
      }
      tintColor {colorUseRule=UseColor_Specified, specifiedColor=<rgba 1 1 1 1>}
      uVRegion {bIsValid=false, max=<0 0>, min=<0 0>}
     }
     hoveredThumbImage {
      bIsDynamicallyLoaded=false
      drawAs=RoundedBox
      imageType=NoImage
      mirroring=NoMirror
      resourceName=None
      resourceObject=null
      tiling=NoTile
      imageSize=<0 0>
      margin=<ltrb 0 0 0 0>
      outlineSettings {
       bUseBrushTransparency=false
       roundingType=FixedRadius
       width=0
       color {colorUseRule=UseColor_Specified, specifiedColor=<rgba 0 0 0 0>}
       cornerRadii=<4 4 4 4>
      }
      tintColor {colorUseRule=UseColor_Specified, specifiedColor=<rgba 0.2159 0.2159 0.2159 1>}
      uVRegion {bIsValid=false, max=<0 0>, min=<0 0>}
     }
     normalThumbImage {
      bIsDynamicallyLoaded=false
      drawAs=RoundedBox
      imageType=NoImage
      mirroring=NoMirror
      resourceName=None
      resourceObject=null
      tiling=NoTile
      imageSize=<0 0>
      margin=<ltrb 0 0 0 0>
      outlineSettings {
       bUseBrushTransparency=false
       roundingType=FixedRadius
       width=0
       color {colorUseRule=UseColor_Specified, specifiedColor=<rgba 0 0 0 0>}
       cornerRadii=<4 4 4 4>
      }
      tintColor {colorUseRule=UseColor_Specified, specifiedColor=<rgba 0.0953 0.0953 0.0953 1>}
      uVRegion {bIsValid=false, max=<0 0>, min=<0 0>}
     }
     verticalBackgroundImage {
      bIsDynamicallyLoaded=false
      drawAs=NoDrawType
      imageType=NoImage
      mirroring=NoMirror
      resourceName=None
      resourceObject=null
      tiling=NoTile
      imageSize=<0 0>
      margin=<ltrb 0 0 0 0>
      outlineSettings {
       bUseBrushTransparency=false
       roundingType=HalfHeightRadius
       width=0
       color {colorUseRule=UseColor_Specified, specifiedColor=<rgba 0 0 0 0>}
       cornerRadii=<0 0 0 0>
      }
      tintColor {colorUseRule=UseColor_Specified, specifiedColor=<rgba 1 1 1 1>}
      uVRegion {bIsValid=false, max=<0 0>, min=<0 0>}
     }
     verticalBottomSlotImage {
      bIsDynamicallyLoaded=false
      drawAs=NoDrawType
      imageType=NoImage
      mirroring=NoMirror
      resourceName=None
      resourceObject=null
      tiling=NoTile
      imageSize=<0 0>
      margin=<ltrb 0 0 0 0>
      outlineSettings {
       bUseBrushTransparency=false
       roundingType=HalfHeightRadius
       width=0
       color {colorUseRule=UseColor_Specified, specifiedColor=<rgba 0 0 0 0>}
       cornerRadii=<0 0 0 0>
      }
      tintColor {colorUseRule=UseColor_Specified, specifiedColor=<rgba 1 1 1 1>}
      uVRegion {bIsValid=false, max=<0 0>, min=<0 0>}
     }
     verticalTopSlotImage {
      bIsDynamicallyLoaded=false
      drawAs=NoDrawType
      imageType=NoImage
      mirroring=NoMirror
      resourceName=None
      resourceObject=null
      tiling=NoTile
      imageSize=<0 0>
      margin=<ltrb 0 0 0 0>
      outlineSettings {
       bUseBrushTransparency=false
       roundingType=HalfHeightRadius
       width=0
       color {colorUseRule=UseColor_Specified, specifiedColor=<rgba 0 0 0 0>}
       cornerRadii=<0 0 0 0>
      }
      tintColor {colorUseRule=UseColor_Specified, specifiedColor=<rgba 1 1 1 1>}
      uVRegion {bIsValid=false, max=<0 0>, min=<0 0>}
     }
    }
    widgetStyle {
     barThickness=8
     bottomShadowBrush {
      bIsDynamicallyLoaded=false
      drawAs=Box
      imageType=FullColor
      mirroring=NoMirror
      resourceName=../../../Engine/Content/Slate/Common/ScrollBoxShadowBottom.png
      resourceObject=null
      tiling=NoTile
      imageSize=<16 8>
      margin=<ltrb 0.5 0 0.5 1>
      outlineSettings {
       bUseBrushTransparency=false
       roundingType=HalfHeightRadius
       width=0
       color {colorUseRule=UseColor_Specified, specifiedColor=<rgba 0 0 0 0>}
       cornerRadii=<0 0 0 0>
      }
      tintColor {colorUseRule=UseColor_Specified, specifiedColor=<rgba 1 1 1 1>}
      uVRegion {bIsValid=false, max=<0 0>, min=<0 0>}
     }
     horizontalScrolledContentPadding=<ltrb 0 0 1 0>
     leftShadowBrush {
      bIsDynamicallyLoaded=false
      drawAs=Box
      imageType=FullColor
      mirroring=NoMirror
      resourceName=../../../Engine/Content/Slate/Common/ScrollBoxShadowLeft.png
      resourceObject=null
      tiling=NoTile
      imageSize=<8 16>
      margin=<ltrb 1 0.5 0 0.5>
      outlineSettings {
       bUseBrushTransparency=false
       roundingType=HalfHeightRadius
       width=0
       color {colorUseRule=UseColor_Specified, specifiedColor=<rgba 0 0 0 0>}
       cornerRadii=<0 0 0 0>
      }
      tintColor {colorUseRule=UseColor_Specified, specifiedColor=<rgba 1 1 1 1>}
      uVRegion {bIsValid=false, max=<0 0>, min=<0 0>}
     }
     rightShadowBrush {
      bIsDynamicallyLoaded=false
      drawAs=Box
      imageType=FullColor
      mirroring=NoMirror
      resourceName=../../../Engine/Content/Slate/Common/ScrollBoxShadowRight.png
      resourceObject=null
      tiling=NoTile
      imageSize=<8 16>
      margin=<ltrb 0 0.5 1 0.5>
      outlineSettings {
       bUseBrushTransparency=false
       roundingType=HalfHeightRadius
       width=0
       color {colorUseRule=UseColor_Specified, specifiedColor=<rgba 0 0 0 0>}
       cornerRadii=<0 0 0 0>
      }
      tintColor {colorUseRule=UseColor_Specified, specifiedColor=<rgba 1 1 1 1>}
      uVRegion {bIsValid=false, max=<0 0>, min=<0 0>}
     }
     topShadowBrush {
      bIsDynamicallyLoaded=false
      drawAs=Box
      imageType=FullColor
      mirroring=NoMirror
      resourceName=../../../Engine/Content/Slate/Common/ScrollBoxShadowTop.png
      resourceObject=null
      tiling=NoTile
      imageSize=<16 8>
      margin=<ltrb 0.5 1 0.5 0>
      outlineSettings {
       bUseBrushTransparency=false
       roundingType=HalfHeightRadius
       width=0
       color {colorUseRule=UseColor_Specified, specifiedColor=<rgba 0 0 0 0>}
       cornerRadii=<0 0 0 0>
      }
      tintColor {colorUseRule=UseColor_Specified, specifiedColor=<rgba 1 1 1 1>}
      uVRegion {bIsValid=false, max=<0 0>, min=<0 0>}
     }
     verticalScrolledContentPadding=<ltrb 0 0 0 1>
    }
   }
   slot {class=VerticalBoxSlot, properties {size {sizeRule=Fill, value=1}}}
  }
  DetailsView_105 {
   widgetClassPath=&DetailsView
   parent=ScrollBox_0
   widget=DetailsView_105
   properties {categoriesToShow [nocategory]}
   slot {class=ScrollBoxSlot}
  }
 ]
}
